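# Physics asset for the male medium SnowSoldierFashion body cosmetic, set up for secondary motion on its front collar flaps and zipper.
# Capsules cover the pelvis, thighs and left and right collar chains, while boxes cover the zipper chain.
# Each dyn MSTR anchor joins the pelvis with free swing and twist.
# Each child bone hangs off its anchor with angular limits, tight for the collar and wider in swing for the zipper.
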PhysicsAsset /BRCosmetics/Characters/Player/Male/Medium/Bodies/M_MED_SnowSoldierFashion/Meshes/M_MED_SnowSoldierFashion_Physics
bodies {
 dyn_collar_front_l_01 [{shapeType=Capsule, length=2, radius=2, shapeName=dyn_collar_front_l_01_capsule, center=<3 0 0>}]
 dyn_collar_front_l_MSTR [{shapeType=Capsule, length=2, radius=2, shapeName=dyn_collar_front_l_MSTR_capsule}]
 dyn_collar_front_r_01 [{shapeType=Capsule, length=2, radius=2, shapeName=dyn_collar_front_r_01_capsule, center=<3 0 0>}]
 dyn_collar_front_r_MSTR [{shapeType=Capsule, length=2, radius=2, shapeName=dyn_collar_front_r_MSTR_capsule}]
 dyn_zipper [{shapeType=Box, extentX=3.5401, extentY=0.8391, extentZ=2.1559, shapeName=dyn_zipper_box, center=<2.9069 1.1346 0>}]
 dyn_zipper_MSTR [{shapeType=Box, extentX=13, extentY=3.6087, extentZ=10, shapeName=dyn_zipper_MSTR_box, center=<2.3683 -1.5546 0>}]
 pelvis [
  {shapeType=Capsule, length=13.2935, radius=12.6884}
  {shapeType=Capsule, length=13.9259, radius=12.056, shapeName=pelvis_capsule, center=<-13.7774 0 0>}
 ]
 thigh_l [{shapeType=Capsule, length=19.245, radius=10.0656, center=<-16.6407 0 0>, rotation=<rot 90 0 0>}]
 thigh_r [{shapeType=Capsule, length=15.7181, radius=10.8092, center=<19.1438 0 0>, rotation=<rot 90 0 0>}]
}
constraints [
 {bone1Name=thigh_l, bone2Name=pelvis, swing1LimitDegrees=45, swing2LimitDegrees=45, twistLimitDegrees=45}
 {bone1Name=thigh_r, bone2Name=pelvis, swing1LimitDegrees=45, swing2LimitDegrees=45, twistLimitDegrees=45}
 {
  bone1Name=dyn_zipper_MSTR
  bone2Name=pelvis
  swing1LimitDegrees=45
  swing1Motion=Free
  swing2LimitDegrees=45
  swing2Motion=Free
  twistLimitDegrees=45
  twistMotion=Free
 }
 {bone1Name=dyn_zipper, bone2Name=dyn_zipper_MSTR, swing1LimitDegrees=75, swing2LimitDegrees=45, twistLimitDegrees=5}
 {
  bone1Name=dyn_collar_front_l_MSTR
  bone2Name=pelvis
  swing1LimitDegrees=45
  swing1Motion=Free
  swing2LimitDegrees=45
  swing2Motion=Free
  twistLimitDegrees=45
  twistMotion=Free
 }
 {bone1Name=dyn_collar_front_l_01, bone2Name=dyn_collar_front_l_MSTR, swing1LimitDegrees=10, swing2LimitDegrees=10, twistLimitDegrees=2.5}
 {
  bone1Name=dyn_collar_front_r_MSTR
  bone2Name=pelvis
  swing1LimitDegrees=45
  swing1Motion=Free
  swing2LimitDegrees=45
  swing2Motion=Free
  twistLimitDegrees=45
  twistMotion=Free
 }
 {bone1Name=dyn_collar_front_r_01, bone2Name=dyn_collar_front_r_MSTR, swing1LimitDegrees=10, swing2LimitDegrees=10, twistLimitDegrees=2.5}
]
