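# Animation Blueprint for the M_MED_SnowSoldierFashion cosmetic body, adding secondary motion and pose fixes to its pants straps, zipper, collar and jacket.
# The main chain rotation-constrains strap masters to the thighs, then simulates straps with Anim Dynamics.
# A weighted multi-bone constraint places the zipper, followed by a RigidBody pass and zipper dynamics with leg collision limits.
# Corrective bone transforms then adjust collar and jacket, several driven by crouch and ADS arm-rotation variables.
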AnimBlueprint /BRCosmetics/Characters/Player/Male/Medium/Bodies/M_MED_SnowSoldierFashion/Meshes/M_MED_SnowSoldierFashion_AnimBP
target_skeleton=&/M_MED_SnowSoldierFashion_Skeleton
graphs [
 AnimGraph {nodes [
   n1 = Anim Dynamics - Bone: dyn_l_outer_pants_strap(Alpha=1, ExternalForce="0,0,0") {
    angularDampingOverride=0.999
    angularSpringConstant=200
    bAngularSpring=true
    bGravityOverrideInSimSpace=true
    bLinearSpring=true
    bOverrideAngularDamping=true
    bOverrideLinearDamping=true
    bUseGravityOverride=true
    gravityScale=0.35
    linearDampingOverride=1
    linearSpringConstant=75
    simulationSpace=BoneRelative
    alphaBoolBlend {blendOption=Linear}
    boundBone {boneName=dyn_l_outer_pants_strap}
    componentAppliedLinearAccClamp=<1500 1500 1500>
    componentLinearAccScale=<0.1 0.1 0.25>
    componentLinearVelScale=<0.3 0.3 0.4>
    gravityOverride {z=-980}
    physicsBodyDefinitions [{
      sphereCollisionRadius=0
      boundBone {boneName=dyn_l_outer_pants_strap}
      boxExtents=<30 30 30>
      constraintSetup {angularLimitsMax=<2 2 2>, angularLimitsMin=<-2 -2 -2>, angularTarget=<1 0 0>, linearAxesMax=<1 1 2>, linearAxesMin=<-1 -1 0>}
      localJointOffset=<10 0 0>
     }]
    relativeSpaceBone {boneName=root}
   }
   n2 = Anim Dynamics - Bone: dyn_l_bk_pants_strap(ComponentPose=@n1, Alpha=1, ExternalForce="0,0,0") {
    angularDampingOverride=0.999
    angularSpringConstant=200
    bAngularSpring=true
    bGravityOverrideInSimSpace=true
    bLinearSpring=true
    bOverrideAngularDamping=true
    bOverrideLinearDamping=true
    bUseGravityOverride=true
    gravityScale=0.35
    linearDampingOverride=1
    linearSpringConstant=75
    simulationSpace=BoneRelative
    alphaBoolBlend {blendOption=Linear}
    boundBone {boneName=dyn_l_bk_pants_strap}
    componentAppliedLinearAccClamp=<1500 1500 1500>
    componentLinearAccScale=<0.1 0.1 0.25>
    componentLinearVelScale=<0.3 0.3 0.4>
    gravityOverride {z=-980}
    physicsBodyDefinitions [{
      sphereCollisionRadius=0
      boundBone {boneName=dyn_l_bk_pants_strap}
      boxExtents=<30 30 30>
      constraintSetup {angularLimitsMax=<2 2 2>, angularLimitsMin=<-2 -2 -2>, angularTarget=<1 0 0>, linearAxesMax=<1 1 2>, linearAxesMin=<-1 -1 0>}
      localJointOffset=<10 0 0>
     }]
    relativeSpaceBone {boneName=root}
   }
   n3 = Anim Dynamics - Bone: dyn_l_fr_pants_strap(ComponentPose=@n2, Alpha=1, ExternalForce="0,0,0") {
    angularDampingOverride=0.999
    angularSpringConstant=200
    bAngularSpring=true
    bGravityOverrideInSimSpace=true
    bLinearSpring=true
    bOverrideAngularDamping=true
    bOverrideLinearDamping=true
    bUseGravityOverride=true
    gravityScale=0.35
    linearDampingOverride=1
    linearSpringConstant=75
    simulationSpace=BoneRelative
    alphaBoolBlend {blendOption=Linear}
    boundBone {boneName=dyn_l_fr_pants_strap}
    componentAppliedLinearAccClamp=<1500 1500 1500>
    componentLinearAccScale=<0.1 0.1 0.25>
    componentLinearVelScale=<0.3 0.3 0.4>
    gravityOverride {z=-980}
    physicsBodyDefinitions [{
      sphereCollisionRadius=0
      boundBone {boneName=dyn_l_fr_pants_strap}
      boxExtents=<30 30 30>
      constraintSetup {angularLimitsMax=<2 2 2>, angularLimitsMin=<-2 -2 -2>, angularTarget=<1 0 0>, linearAxesMax=<1 1 2>, linearAxesMin=<-1 -1 0>}
      localJointOffset=<10 0 0>
     }]
    relativeSpaceBone {boneName=root}
   }
   n4 = Anim Dynamics - Bone: dyn_l_inner_pants_strap(ComponentPose=@n3, Alpha=1, ExternalForce="0,0,0") {
    angularDampingOverride=0.999
    angularSpringConstant=200
    bAngularSpring=true
    bGravityOverrideInSimSpace=true
    bLinearSpring=true
    bOverrideAngularDamping=true
    bOverrideLinearDamping=true
    bUseGravityOverride=true
    gravityScale=0.35
    linearDampingOverride=1
    linearSpringConstant=75
    simulationSpace=BoneRelative
    alphaBoolBlend {blendOption=Linear}
    boundBone {boneName=dyn_l_inner_pants_strap}
    componentAppliedLinearAccClamp=<1500 1500 1500>
    componentLinearAccScale=<0.1 0.1 0.25>
    componentLinearVelScale=<0.3 0.3 0.4>
    gravityOverride {z=-980}
    physicsBodyDefinitions [{
      sphereCollisionRadius=0
      boundBone {boneName=dyn_l_inner_pants_strap}
      boxExtents=<30 30 30>
      constraintSetup {angularLimitsMax=<2 2 2>, angularLimitsMin=<-2 -2 -2>, angularTarget=<1 0 0>, linearAxesMax=<1 1 2>, linearAxesMin=<-1 -1 0>}
      localJointOffset=<10 0 0>
     }]
    relativeSpaceBone {boneName=root}
   }
   n5 = Anim Dynamics - Bone: dyn_r_outer_pants_strap(ComponentPose=@n4, Alpha=1, ExternalForce="0,0,0") {
    angularDampingOverride=0.999
    angularSpringConstant=200
    bAngularSpring=true
    bGravityOverrideInSimSpace=true
    bLinearSpring=true
    bOverrideAngularDamping=true
    bOverrideLinearDamping=true
    bUseGravityOverride=true
    gravityScale=0.35
    linearDampingOverride=1
    linearSpringConstant=75
    simulationSpace=BoneRelative
    alphaBoolBlend {blendOption=Linear}
    boundBone {boneName=dyn_r_outer_pants_strap}
    componentAppliedLinearAccClamp=<1500 1500 1500>
    componentLinearAccScale=<0.1 0.1 0.25>
    componentLinearVelScale=<0.3 0.3 0.4>
    gravityOverride {z=-980}
    physicsBodyDefinitions [{
      sphereCollisionRadius=0
      boundBone {boneName=dyn_r_outer_pants_strap}
      boxExtents=<30 30 30>
      constraintSetup {angularLimitsMax=<2 2 2>, angularLimitsMin=<-2 -2 -2>, angularTarget=<1 0 0>, linearAxesMax=<1 1 2>, linearAxesMin=<-1 -1 0>}
      localJointOffset=<10 0 0>
     }]
    relativeSpaceBone {boneName=root}
   }
   n6 = Anim Dynamics - Bone: dyn_r_inner_pants_strap(ComponentPose=@n5, Alpha=1, ExternalForce="0,0,0") {
    angularDampingOverride=0.999
    angularSpringConstant=200
    bAngularSpring=true
    bGravityOverrideInSimSpace=true
    bLinearSpring=true
    bOverrideAngularDamping=true
    bOverrideLinearDamping=true
    bUseGravityOverride=true
    gravityScale=0.35
    linearDampingOverride=1
    linearSpringConstant=75
    simulationSpace=BoneRelative
    alphaBoolBlend {blendOption=Linear}
    boundBone {boneName=dyn_r_inner_pants_strap}
    componentAppliedLinearAccClamp=<1500 1500 1500>
    componentLinearAccScale=<0.1 0.1 0.25>
    componentLinearVelScale=<0.3 0.3 0.4>
    gravityOverride {z=-980}
    physicsBodyDefinitions [{
      sphereCollisionRadius=0
      boundBone {boneName=dyn_r_inner_pants_strap}
      boxExtents=<30 30 30>
      constraintSetup {angularLimitsMax=<2 2 2>, angularLimitsMin=<-2 -2 -2>, angularTarget=<1 0 0>, linearAxesMax=<1 1 2>, linearAxesMin=<-1 -1 0>}
      localJointOffset=<10 0 0>
     }]
    relativeSpaceBone {boneName=root}
   }
   n7 = Anim Dynamics - Bone: dyn_r_bk_pants_strap(ComponentPose=@n6, Alpha=1, ExternalForce="0,0,0") {
    angularDampingOverride=0.999
    angularSpringConstant=200
    bAngularSpring=true
    bGravityOverrideInSimSpace=true
    bLinearSpring=true
    bOverrideAngularDamping=true
    bOverrideLinearDamping=true
    bUseGravityOverride=true
    gravityScale=0.35
    linearDampingOverride=1
    linearSpringConstant=75
    simulationSpace=BoneRelative
    alphaBoolBlend {blendOption=Linear}
    boundBone {boneName=dyn_r_bk_pants_strap}
    componentAppliedLinearAccClamp=<1500 1500 1500>
    componentLinearAccScale=<0.1 0.1 0.25>
    componentLinearVelScale=<0.3 0.3 0.4>
    gravityOverride {z=-980}
    physicsBodyDefinitions [{
      sphereCollisionRadius=0
      boundBone {boneName=dyn_r_bk_pants_strap}
      boxExtents=<30 30 30>
      constraintSetup {angularLimitsMax=<2 2 2>, angularLimitsMin=<-2 -2 -2>, angularTarget=<1 0 0>, linearAxesMax=<1 1 2>, linearAxesMin=<-1 -1 0>}
      localJointOffset=<10 0 0>
     }]
    relativeSpaceBone {boneName=root}
   }
   Anim Dynamics - Bone: dyn_r_fr_pants_strap(ComponentPose=@n7, Alpha=1, ExternalForce="0,0,0") {
    angularDampingOverride=0.999
    angularSpringConstant=200
    bAngularSpring=true
    bGravityOverrideInSimSpace=true
    bLinearSpring=true
    bOverrideAngularDamping=true
    bOverrideLinearDamping=true
    bUseGravityOverride=true
    gravityScale=0.35
    linearDampingOverride=1
    linearSpringConstant=75
    simulationSpace=BoneRelative
    alphaBoolBlend {blendOption=Linear}
    boundBone {boneName=dyn_r_fr_pants_strap}
    componentAppliedLinearAccClamp=<1500 1500 1500>
    componentLinearAccScale=<0.1 0.1 0.25>
    componentLinearVelScale=<0.3 0.3 0.4>
    gravityOverride {z=-980}
    physicsBodyDefinitions [{
      sphereCollisionRadius=0
      boundBone {boneName=dyn_r_fr_pants_strap}
      boxExtents=<30 30 30>
      constraintSetup {angularLimitsMax=<2 2 2>, angularLimitsMin=<-2 -2 -2>, angularTarget=<1 0 0>, linearAxesMax=<1 1 2>, linearAxesMin=<-1 -1 0>}
      localJointOffset=<10 0 0>
     }]
    relativeSpaceBone {boneName=root}
   }
   n8 = "Transform (Modify) Bone - Bone: dyn_l_outer_pants_strap_mstr"(ComponentPose=Local To Component(LocalPose=Use cached pose 'PreCrouch'()), Alpha=1, Translation="-8,0,1", Rotation="0,0,0", Scale="1,1,1") {
    rotationMode=BMM_Additive
    rotationSpace=BCS_BoneSpace
    translationMode=BMM_Additive
    translationSpace=BCS_BoneSpace
    alphaBoolBlend {blendOption=Linear}
    boneToModify {boneName=dyn_l_outer_pants_strap_mstr}
    translation {x=-8, z=1}
   }
   n9 = "Transform (Modify) Bone - Bone: dyn_l_bk_pants_strap_mstr"(ComponentPose=@n8, Alpha=1, Translation="-10,0,1", Rotation="0,0,0", Scale="1,1,1") {
    rotationMode=BMM_Additive
    rotationSpace=BCS_BoneSpace
    translationMode=BMM_Additive
    translationSpace=BCS_BoneSpace
    alphaBoolBlend {blendOption=Linear}
    boneToModify {boneName=dyn_l_bk_pants_strap_mstr}
    translation {x=-10, z=1}
   }
   n10 = "Transform (Modify) Bone - Bone: dyn_l_fr_pants_strap_mstr"(ComponentPose=@n9, Alpha=1, Translation="-6,0,1", Rotation="0,0,0", Scale="1,1,1") {
    rotationMode=BMM_Additive
    rotationSpace=BCS_BoneSpace
    translationMode=BMM_Additive
    translationSpace=BCS_BoneSpace
    alphaBoolBlend {blendOption=Linear}
    boneToModify {boneName=dyn_l_fr_pants_strap_mstr}
    translation {x=-6, z=1}
   }
   n11 = "Transform (Modify) Bone - Bone: dyn_l_inner_pants_strap_mstr"(ComponentPose=@n10, Alpha=1, Translation="-8,0,1", Rotation="0,0,0", Scale="1,1,1") {
    rotationMode=BMM_Additive
    rotationSpace=BCS_BoneSpace
    translationMode=BMM_Additive
    translationSpace=BCS_BoneSpace
    alphaBoolBlend {blendOption=Linear}
    boneToModify {boneName=dyn_l_inner_pants_strap_mstr}
    translation {x=-8, z=1}
   }
   n12 = "Transform (Modify) Bone - Bone: dyn_r_outer_pants_strap_mstr"(ComponentPose=@n11, Alpha=1, Translation="-8,0,0.01", Rotation="0,0,0", Scale="1,1,1") {
    rotationMode=BMM_Additive
    rotationSpace=BCS_BoneSpace
    translationMode=BMM_Additive
    translationSpace=BCS_BoneSpace
    alphaBoolBlend {blendOption=Linear}
    boneToModify {boneName=dyn_r_outer_pants_strap_mstr}
    translation {x=-8, z=0.01}
   }
   n13 = "Transform (Modify) Bone - Bone: dyn_r_inner_pants_strap_mstr"(ComponentPose=@n12, Alpha=1, Translation="-6.5,0,1.5", Rotation="0,0,0", Scale="1,1,1") {
    rotationMode=BMM_Additive
    rotationSpace=BCS_BoneSpace
    translationMode=BMM_Additive
    translationSpace=BCS_BoneSpace
    alphaBoolBlend {blendOption=Linear}
    boneToModify {boneName=dyn_r_inner_pants_strap_mstr}
    translation {x=-6.5, z=1.5}
   }
   n14 = "Transform (Modify) Bone - Bone: dyn_r_bk_pants_strap_mstr"(ComponentPose=@n13, Alpha=1, Translation="-10,0,1", Rotation="0,0,0", Scale="1,1,1") {
    rotationMode=BMM_Additive
    rotationSpace=BCS_BoneSpace
    translationMode=BMM_Additive
    translationSpace=BCS_BoneSpace
    alphaBoolBlend {blendOption=Linear}
    boneToModify {boneName=dyn_r_bk_pants_strap_mstr}
    translation {x=-10, z=1}
   }
   n15 = "Transform (Modify) Bone - Bone: dyn_r_fr_pants_strap_mstr"(ComponentPose=@n14, Alpha=1, Translation="-6,0,1", Rotation="0,0,0", Scale="1,1,1") {
    rotationMode=BMM_Additive
    rotationSpace=BCS_BoneSpace
    translationMode=BMM_Additive
    translationSpace=BCS_BoneSpace
    alphaBoolBlend {blendOption=Linear}
    boneToModify {boneName=dyn_r_fr_pants_strap_mstr}
    translation {x=-6, z=1}
   }
   Blend Poses by bool(BlendPose_0=Component To Local(ComponentPose=@n15), BlendTime_0=0.3, BlendTime_1=0.2, bActiveValue=Get bIsCrouching()) {blendType=Linear}
   n16 = "Transform (Modify) Bone - Bone: dyn_l_outer_pants_strap_mstr"(ComponentPose=Local To Component(LocalPose=Use cached pose 'Pre_Crouch'()), Alpha=1, Translation="-8,0,1", Rotation="0,0,0", Scale="1,1,1") {
    rotationMode=BMM_Additive
    rotationSpace=BCS_BoneSpace
    translationMode=BMM_Additive
    translationSpace=BCS_BoneSpace
    alphaBoolBlend {blendOption=Linear}
    boneToModify {boneName=dyn_l_outer_pants_strap_mstr}
    translation {x=-8, z=1}
   }
   n17 = "Transform (Modify) Bone - Bone: dyn_l_bk_pants_strap_mstr"(ComponentPose=@n16, Alpha=1, Translation="-10,0,1", Rotation="0,0,0", Scale="1,1,1") {
    rotationMode=BMM_Additive
    rotationSpace=BCS_BoneSpace
    translationMode=BMM_Additive
    translationSpace=BCS_BoneSpace
    alphaBoolBlend {blendOption=Linear}
    boneToModify {boneName=dyn_l_bk_pants_strap_mstr}
    translation {x=-10, z=1}
   }
   n18 = "Transform (Modify) Bone - Bone: dyn_l_fr_pants_strap_mstr"(ComponentPose=@n17, Alpha=1, Translation="-6,0,1", Rotation="0,0,0", Scale="1,1,1") {
    rotationMode=BMM_Additive
    rotationSpace=BCS_BoneSpace
    translationMode=BMM_Additive
    translationSpace=BCS_BoneSpace
    alphaBoolBlend {blendOption=Linear}
    boneToModify {boneName=dyn_l_fr_pants_strap_mstr}
    translation {x=-6, z=1}
   }
   n19 = "Transform (Modify) Bone - Bone: dyn_l_inner_pants_strap_mstr"(ComponentPose=@n18, Alpha=1, Translation="-8,0,1", Rotation="0,0,0", Scale="1,1,1") {
    rotationMode=BMM_Additive
    rotationSpace=BCS_BoneSpace
    translationMode=BMM_Additive
    translationSpace=BCS_BoneSpace
    alphaBoolBlend {blendOption=Linear}
    boneToModify {boneName=dyn_l_inner_pants_strap_mstr}
    translation {x=-8, z=1}
   }
   n20 = "Transform (Modify) Bone - Bone: dyn_r_outer_pants_strap_mstr"(ComponentPose=@n19, Alpha=1, Translation="-8,0,0.01", Rotation="0,0,0", Scale="1,1,1") {
    rotationMode=BMM_Additive
    rotationSpace=BCS_BoneSpace
    translationMode=BMM_Additive
    translationSpace=BCS_BoneSpace
    alphaBoolBlend {blendOption=Linear}
    boneToModify {boneName=dyn_r_outer_pants_strap_mstr}
    translation {x=-8, z=0.01}
   }
   n21 = "Transform (Modify) Bone - Bone: dyn_r_inner_pants_strap_mstr"(ComponentPose=@n20, Alpha=1, Translation="-6.5,0,1.5", Rotation="0,0,0", Scale="1,1,1") {
    rotationMode=BMM_Additive
    rotationSpace=BCS_BoneSpace
    translationMode=BMM_Additive
    translationSpace=BCS_BoneSpace
    alphaBoolBlend {blendOption=Linear}
    boneToModify {boneName=dyn_r_inner_pants_strap_mstr}
    translation {x=-6.5, z=1.5}
   }
   n22 = "Transform (Modify) Bone - Bone: dyn_r_bk_pants_strap_mstr"(ComponentPose=@n21, Alpha=1, Translation="-10,0,1", Rotation="0,0,0", Scale="1,1,1") {
    rotationMode=BMM_Additive
    rotationSpace=BCS_BoneSpace
    translationMode=BMM_Additive
    translationSpace=BCS_BoneSpace
    alphaBoolBlend {blendOption=Linear}
    boneToModify {boneName=dyn_r_bk_pants_strap_mstr}
    translation {x=-10, z=1}
   }
   n23 = "Transform (Modify) Bone - Bone: dyn_r_fr_pants_strap_mstr"(ComponentPose=@n22, Alpha=1, Translation="-6,0,1", Rotation="0,0,0", Scale="1,1,1") {
    rotationMode=BMM_Additive
    rotationSpace=BCS_BoneSpace
    translationMode=BMM_Additive
    translationSpace=BCS_BoneSpace
    alphaBoolBlend {blendOption=Linear}
    boneToModify {boneName=dyn_r_fr_pants_strap_mstr}
    translation {x=-6, z=1}
   }
   Blend Poses by bool(
    BlendPose_0=Component To Local(ComponentPose=@n23)
    BlendPose_1=Use cached pose 'Pre_Crouch'()
    BlendTime_0=0.2
    BlendTime_1=0.2
    bActiveValue=Get bIsCrouching()
   ) {blendType=Linear}
   n24 = Get head_rot_back() -> (head_rot_back:double)
   n25 = "Transform (Modify) Bone - Bone: dyn_collar_back_l_01"!(ComponentPose=Local To Component(), Alpha=@n24, Translation="0,-1,0", Rotation="0,0,-20") -> (Pose) {
    rotationMode=BMM_Additive
    rotationSpace=BCS_BoneSpace
    scaleSpace=BCS_BoneSpace
    translationMode=BMM_Additive
    translationSpace=BCS_BoneSpace
    alphaBoolBlend {blendOption=Linear}
    boneToModify {boneName=dyn_collar_back_l_01}
    rotation {roll=-20}
    scale=<2 2 2>
    translation {y=-1}
   }
   n26 = "Transform (Modify) Bone - Bone: dyn_collar_back_r_01"!(ComponentPose=@n25, Alpha=@n24, Translation="2,-2,0", Rotation="0,-20,-45") -> (Pose) {
    rotationMode=BMM_Additive
    rotationSpace=BCS_BoneSpace
    scaleSpace=BCS_BoneSpace
    translationMode=BMM_Additive
    translationSpace=BCS_BoneSpace
    alphaBoolBlend {blendOption=Linear}
    boneToModify {boneName=dyn_collar_back_r_01}
    rotation {roll=-45, yaw=-20}
    scale=<2 2 2>
    translation=<2 -2>
   }
   n27 = Get Head_Rot_FAR_Back() -> (Head_Rot_FAR_Back:double)
   n28 = "Transform (Modify) Bone - Bone: dyn_collar_back_l_01"(ComponentPose=@n26, Alpha=@n27, Translation="4,-6,0", Rotation="-30,0,-40") {
    rotationMode=BMM_Additive
    rotationSpace=BCS_BoneSpace
    scaleSpace=BCS_BoneSpace
    translationMode=BMM_Additive
    translationSpace=BCS_BoneSpace
    alphaBoolBlend {blendOption=Linear}
    boneToModify {boneName=dyn_collar_back_l_01}
    rotation {pitch=-30, roll=-40}
    scale=<2 2 2>
    translation=<4 -6>
   }
   n29 = "Transform (Modify) Bone - Bone: dyn_collar_back_r_01"(ComponentPose=@n28, Alpha=@n27, Translation="2,-7,0", Rotation="-30,-20,0") {
    rotationMode=BMM_Additive
    rotationSpace=BCS_BoneSpace
    scaleSpace=BCS_BoneSpace
    translationMode=BMM_Additive
    translationSpace=BCS_BoneSpace
    alphaBoolBlend {blendOption=Linear}
    boneToModify {boneName=dyn_collar_back_r_01}
    rotation {pitch=-30, yaw=-20}
    scale=<2 2 2>
    translation=<2 -7>
   }
   Component To Local(ComponentPose=@n29)
   n30 = "Constraint - dyn_jacket_fr_r_mstr (R)"(Alpha=1, ConstraintWeights_0=1) {
    alphaBoolBlend {blendOption=Linear}
    boneToModify {boneName=dyn_jacket_fr_r_mstr}
    constraintSetup [{transformType=Rotation, targetBone {boneName=thigh_r}}]
    constraintWeights [1]
   }
   "Constraint - dyn_jacket_fr_l_mstr (R)"(ComponentPose=@n30, Alpha=1, ConstraintWeights_0=1) {
    alphaBoolBlend {blendOption=Linear}
    boneToModify {boneName=dyn_jacket_fr_l_mstr}
    constraintSetup [{transformType=Rotation, targetBone {boneName=thigh_l}}]
    constraintWeights [1]
   }
   "Constraint - dyn_jacket_fr_r_mstr (P)"(Alpha=1, ConstraintWeights_0=1) {
    alphaBoolBlend {blendOption=Linear}
    boneToModify {boneName=dyn_jacket_fr_r_mstr}
    constraintSetup [{targetBone {boneName=dyn_r_fr_upper_strap}}]
    constraintWeights [1]
   }
   "Constraint - dyn_jacket_fr_l_mstr (P)"(Alpha=1, ConstraintWeights_0=1) {
    alphaBoolBlend {blendOption=Linear}
    boneToModify {boneName=dyn_jacket_fr_l_mstr}
    constraintSetup [{targetBone {boneName=dyn_l_fr_upper_strap}}]
    constraintWeights [1]
   }
   Constraint!(ComponentPose, Alpha=1) -> (Pose) {alphaBoolBlend {blendOption=Linear}, boneToModify=~}
   n31 = "Transform (Modify) Bone - Bone: dyn_collar_front_r_MSTR"(Alpha=Get ADS_UpArm_Rot() -> (ADS_UpArm_Rot:double), Translation="-5,3,0", Rotation="45,-40,0", Scale="1,1,1") {
    rotationMode=BMM_Additive
    rotationSpace=BCS_BoneSpace
    translationMode=BMM_Additive
    translationSpace=BCS_BoneSpace
    alphaBoolBlend {blendOption=Linear}
    boneToModify {boneName=dyn_collar_front_r_MSTR}
    rotation {pitch=45, yaw=-40}
    translation=<-5 3>
   }
   "Transform (Modify) Bone - Bone: dyn_collar_front_l_MSTR"(ComponentPose=@n31, Alpha=Get ADS_UpArm_Rot_R() -> (ADS_UpArm_Rot_R:double), Translation="-2,2,0", Rotation="0,0,0", Scale="1,1,1") {
    rotationMode=BMM_Additive
    rotationSpace=BCS_BoneSpace
    translationMode=BMM_Additive
    translationSpace=BCS_BoneSpace
    alphaBoolBlend {blendOption=Linear}
    boneToModify {boneName=dyn_collar_front_l_MSTR}
    translation=<-2 2>
   }
   "Transform (Modify) Bone"(Alpha=1, Translation="0,0,0", Rotation="0,0,0", Scale="1,1,1") {alphaBoolBlend {blendOption=Linear}, boneToModify=~}
   "Transform (Modify) Bone - Bone: dyn_collar_front_r_MSTR"(Alpha=1, Translation="-4.2,3,0", Rotation="20,-40,0", Scale="1,1,1") {
    rotationMode=BMM_Additive
    rotationSpace=BCS_BoneSpace
    translationMode=BMM_Additive
    translationSpace=BCS_BoneSpace
    alphaBoolBlend {blendOption=Linear}
    boneToModify {boneName=dyn_collar_front_r_MSTR}
    rotation {pitch=20, yaw=-40}
    translation=<-4.2 3>
   }
   "Transform (Modify) Bone - Bone: dyn_jacket_bk_mstr"(Alpha=1, Translation="-5,50,0", Rotation="0,0,0", Scale="1,1,1") {
    rotationMode=BMM_Additive
    rotationSpace=BCS_BoneSpace
    translationMode=BMM_Additive
    translationSpace=BCS_BoneSpace
    alphaBoolBlend {blendOption=Linear}
    boneToModify {boneName=dyn_jacket_bk_mstr}
    translation=<-5 50>
   }
   n32 = "Constraint - dyn_r_outer_pants_strap_mstr (R)"(ComponentPose=Local To Component(LocalPose=Use cached pose 'PreCrouch'()), Alpha=1, ConstraintWeights_0=1) {
    alphaBoolBlend {blendOption=Linear}
    boneToModify {boneName=dyn_r_outer_pants_strap_mstr}
    constraintSetup [{transformType=Rotation, targetBone {boneName=thigh_r}}]
    constraintWeights [1]
   }
   n33 = "Constraint - dyn_r_inner_pants_strap_mstr (R)"(ComponentPose=@n32, Alpha=1, ConstraintWeights_0=1) {
    alphaBoolBlend {blendOption=Linear}
    boneToModify {boneName=dyn_r_inner_pants_strap_mstr}
    constraintSetup [{transformType=Rotation, targetBone {boneName=thigh_r}}]
    constraintWeights [1]
   }
   n34 = "Constraint - dyn_r_bk_pants_strap_mstr (R)"(ComponentPose=@n33, Alpha=1, ConstraintWeights_0=1) {
    alphaBoolBlend {blendOption=Linear}
    boneToModify {boneName=dyn_r_bk_pants_strap_mstr}
    constraintSetup [{transformType=Rotation, targetBone {boneName=thigh_r}}]
    constraintWeights [1]
   }
   n35 = "Constraint - dyn_r_fr_pants_strap_mstr (R)"(ComponentPose=@n34, Alpha=1, ConstraintWeights_0=1) {
    alphaBoolBlend {blendOption=Linear}
    boneToModify {boneName=dyn_r_fr_pants_strap_mstr}
    constraintSetup [{transformType=Rotation, targetBone {boneName=thigh_r}}]
    constraintWeights [1]
   }
   n36 = "Constraint - dyn_l_outer_pants_strap_mstr (R)"(ComponentPose=@n35, Alpha=1, ConstraintWeights_0=1) {
    alphaBoolBlend {blendOption=Linear}
    boneToModify {boneName=dyn_l_outer_pants_strap_mstr}
    constraintSetup [{transformType=Rotation, targetBone {boneName=thigh_l}}]
    constraintWeights [1]
   }
   n37 = "Constraint - dyn_l_bk_pants_strap_mstr (R)"(ComponentPose=@n36, Alpha=1, ConstraintWeights_0=1) {
    alphaBoolBlend {blendOption=Linear}
    boneToModify {boneName=dyn_l_bk_pants_strap_mstr}
    constraintSetup [{transformType=Rotation, targetBone {boneName=thigh_l}}]
    constraintWeights [1]
   }
   n38 = "Constraint - dyn_l_fr_pants_strap_mstr (R)"(ComponentPose=@n37, Alpha=1, ConstraintWeights_0=1) {
    alphaBoolBlend {blendOption=Linear}
    boneToModify {boneName=dyn_l_fr_pants_strap_mstr}
    constraintSetup [{transformType=Rotation, targetBone {boneName=thigh_l}}]
    constraintWeights [1]
   }
   n39 = "Constraint - dyn_l_inner_pants_strap_mstr (R)"(ComponentPose=@n38, Alpha=1, ConstraintWeights_0=1) {
    alphaBoolBlend {blendOption=Linear}
    boneToModify {boneName=dyn_l_inner_pants_strap_mstr}
    constraintSetup [{transformType=Rotation, targetBone {boneName=thigh_l}}]
    constraintWeights [1]
   }
   n40 = Anim Dynamics - Bone: dyn_l_outer_pants_strap(ComponentPose=@n39, Alpha=1, ExternalForce="0,0,0") {
    angularDampingOverride=0.9999
    angularSpringConstant=200
    bGravityOverrideInSimSpace=true
    bOverrideAngularDamping=true
    bOverrideLinearDamping=true
    bUseGravityOverride=true
    gravityScale=0.35
    linearDampingOverride=0.9999
    linearSpringConstant=75
    simulationSpace=BoneRelative
    alphaBoolBlend {blendOption=Linear}
    boundBone {boneName=dyn_l_outer_pants_strap}
    componentAppliedLinearAccClamp=<1500 1500 1500>
    componentLinearAccScale=<0.4 0.4 0.55>
    componentLinearVelScale=<0.4 0.4 0.6>
    gravityOverride {x=-980}
    physicsBodyDefinitions [{
      sphereCollisionRadius=0
      boundBone {boneName=dyn_l_outer_pants_strap}
      boxExtents=<12 12 12>
      constraintSetup {angularLimitsMax=<5 5 5>, angularLimitsMin=<-5 -5 -5>, angularTarget=<1 0 0>, linearAxesMax=<1.5 1.5 2>, linearAxesMin=<-1.5 -1.5 0>}
      localJointOffset=<5 0 0>
     }]
    relativeSpaceBone {boneName=pelvis}
   }
   n41 = Anim Dynamics - Bone: dyn_l_bk_pants_strap(ComponentPose=@n40, Alpha=1, ExternalForce="0,0,0") {
    angularDampingOverride=0.9999
    angularSpringConstant=200
    bGravityOverrideInSimSpace=true
    bOverrideAngularDamping=true
    bOverrideLinearDamping=true
    bUseGravityOverride=true
    gravityScale=0.35
    linearDampingOverride=0.9999
    linearSpringConstant=75
    simulationSpace=BoneRelative
    alphaBoolBlend {blendOption=Linear}
    boundBone {boneName=dyn_l_bk_pants_strap}
    componentAppliedLinearAccClamp=<1500 1500 1500>
    componentLinearAccScale=<0.4 0.4 0.55>
    componentLinearVelScale=<0.4 0.4 0.6>
    gravityOverride {x=-980}
    physicsBodyDefinitions [{
      sphereCollisionRadius=0
      boundBone {boneName=dyn_l_bk_pants_strap}
      boxExtents=<12 12 12>
      constraintSetup {angularLimitsMax=<5 5 5>, angularLimitsMin=<-5 -5 -5>, angularTarget=<1 0 0>, linearAxesMax=<1.5 1.5 2>, linearAxesMin=<-1.5 -1.5 0>}
      localJointOffset=<5 0 0>
     }]
    relativeSpaceBone {boneName=pelvis}
   }
   n42 = Anim Dynamics - Bone: dyn_l_fr_pants_strap(ComponentPose=@n41, Alpha=1, ExternalForce="0,0,0") {
    angularDampingOverride=0.9999
    angularSpringConstant=200
    bGravityOverrideInSimSpace=true
    bOverrideAngularDamping=true
    bOverrideLinearDamping=true
    bUseGravityOverride=true
    gravityScale=0.35
    linearDampingOverride=0.9999
    linearSpringConstant=75
    simulationSpace=BoneRelative
    alphaBoolBlend {blendOption=Linear}
    boundBone {boneName=dyn_l_fr_pants_strap}
    componentAppliedLinearAccClamp=<1500 1500 1500>
    componentLinearAccScale=<0.4 0.4 0.55>
    componentLinearVelScale=<0.4 0.4 0.6>
    gravityOverride {x=-980}
    physicsBodyDefinitions [{
      sphereCollisionRadius=0
      boundBone {boneName=dyn_l_fr_pants_strap}
      boxExtents=<12 12 12>
      constraintSetup {angularLimitsMax=<5 5 5>, angularLimitsMin=<-5 -5 -5>, angularTarget=<1 0 0>, linearAxesMax=<1.5 1.5 2>, linearAxesMin=<-1.5 -1.5 0>}
      localJointOffset=<5 0 0>
     }]
    relativeSpaceBone {boneName=pelvis}
   }
   n43 = Anim Dynamics - Bone: dyn_l_inner_pants_strap(ComponentPose=@n42, Alpha=1, ExternalForce="0,0,0") {
    angularDampingOverride=0.9999
    angularSpringConstant=200
    bGravityOverrideInSimSpace=true
    bOverrideAngularDamping=true
    bOverrideLinearDamping=true
    bUseGravityOverride=true
    gravityScale=0.35
    linearDampingOverride=0.9999
    linearSpringConstant=75
    simulationSpace=BoneRelative
    alphaBoolBlend {blendOption=Linear}
    boundBone {boneName=dyn_l_inner_pants_strap}
    componentAppliedLinearAccClamp=<1500 1500 1500>
    componentLinearAccScale=<0.4 0.4 0.55>
    componentLinearVelScale=<0.4 0.4 0.6>
    gravityOverride {x=-980}
    physicsBodyDefinitions [{
      sphereCollisionRadius=0
      boundBone {boneName=dyn_l_inner_pants_strap}
      boxExtents=<12 12 12>
      constraintSetup {angularLimitsMax=<5 5 5>, angularLimitsMin=<-5 -5 -5>, angularTarget=<1 0 0>, linearAxesMax=<1.5 1.5 2>, linearAxesMin=<-1.5 -1.5 0>}
      localJointOffset=<5 0 0>
     }]
    relativeSpaceBone {boneName=pelvis}
   }
   n44 = Anim Dynamics - Bone: dyn_r_outer_pants_strap(ComponentPose=@n43, Alpha=1, ExternalForce="0,0,0") {
    angularDampingOverride=0.9999
    angularSpringConstant=75
    bGravityOverrideInSimSpace=true
    bOverrideAngularDamping=true
    bOverrideLinearDamping=true
    bUseGravityOverride=true
    gravityScale=0.35
    linearDampingOverride=0.9999
    linearSpringConstant=75
    simulationSpace=BoneRelative
    alphaBoolBlend {blendOption=Linear}
    boundBone {boneName=dyn_r_outer_pants_strap}
    componentAppliedLinearAccClamp=<1500 1500 1500>
    componentLinearAccScale=<0.4 0.4 0.55>
    componentLinearVelScale=<0.4 0.4 0.6>
    gravityOverride {x=-980}
    physicsBodyDefinitions [{
      sphereCollisionRadius=0
      boundBone {boneName=dyn_r_outer_pants_strap}
      boxExtents=<12 12 12>
      constraintSetup {angularLimitsMax=<5 5 5>, angularLimitsMin=<-5 -5 -5>, angularTarget=<1 0 0>, linearAxesMax=<1.5 1.5 2>, linearAxesMin=<-1.5 -1.5 0>}
      localJointOffset=<5 0 -2>
     }]
    relativeSpaceBone {boneName=pelvis}
   }
   n45 = Anim Dynamics - Bone: dyn_r_inner_pants_strap(ComponentPose=@n44, Alpha=1, ExternalForce="0,0,0") {
    angularDampingOverride=0.9999
    angularSpringConstant=75
    bGravityOverrideInSimSpace=true
    bOverrideAngularDamping=true
    bOverrideLinearDamping=true
    bUseGravityOverride=true
    gravityScale=0.35
    linearDampingOverride=0.9999
    linearSpringConstant=75
    simulationSpace=BoneRelative
    alphaBoolBlend {blendOption=Linear}
    boundBone {boneName=dyn_r_inner_pants_strap}
    componentAppliedLinearAccClamp=<1500 1500 1500>
    componentLinearAccScale=<0.4 0.4 0.55>
    componentLinearVelScale=<0.4 0.4 0.6>
    gravityOverride {x=-980}
    physicsBodyDefinitions [{
      sphereCollisionRadius=0
      boundBone {boneName=dyn_r_inner_pants_strap}
      boxExtents=<12 12 12>
      constraintSetup {angularLimitsMax=<5 5 5>, angularLimitsMin=<-5 -5 -5>, angularTarget=<1 0 0>, linearAxesMax=<1.5 1.5 2>, linearAxesMin=<-1.5 -1.5 0>}
      localJointOffset=<5 0 0>
     }]
    relativeSpaceBone {boneName=pelvis}
   }
   n46 = Anim Dynamics - Bone: dyn_r_bk_pants_strap(ComponentPose=@n45, Alpha=1, ExternalForce="0,0,0") {
    angularDampingOverride=0.9999
    angularSpringConstant=75
    bGravityOverrideInSimSpace=true
    bOverrideAngularDamping=true
    bOverrideLinearDamping=true
    bUseGravityOverride=true
    gravityScale=0.35
    linearDampingOverride=0.9999
    linearSpringConstant=75
    simulationSpace=BoneRelative
    alphaBoolBlend {blendOption=Linear}
    boundBone {boneName=dyn_r_bk_pants_strap}
    componentAppliedLinearAccClamp=<1500 1500 1500>
    componentLinearAccScale=<0.4 0.4 0.55>
    componentLinearVelScale=<0.4 0.4 0.6>
    gravityOverride {x=-980}
    physicsBodyDefinitions [{
      sphereCollisionRadius=0
      boundBone {boneName=dyn_r_bk_pants_strap}
      boxExtents=<12 12 12>
      constraintSetup {angularLimitsMax=<5 5 5>, angularLimitsMin=<-5 -5 -5>, angularTarget=<1 0 0>, linearAxesMax=<1.5 1.5 2>, linearAxesMin=<-1.5 -1.5 0>}
      localJointOffset=<5 0 0>
     }]
    relativeSpaceBone {boneName=pelvis}
   }
   n47 = Anim Dynamics - Bone: dyn_r_fr_pants_strap(ComponentPose=@n46, Alpha=1, ExternalForce="0,0,0") {
    angularDampingOverride=0.9999
    angularSpringConstant=100
    bAngularSpring=true
    bGravityOverrideInSimSpace=true
    bOverrideAngularDamping=true
    bOverrideLinearDamping=true
    bUseGravityOverride=true
    gravityScale=0.35
    linearDampingOverride=0.9999
    linearSpringConstant=75
    simulationSpace=BoneRelative
    alphaBoolBlend {blendOption=Linear}
    boundBone {boneName=dyn_r_fr_pants_strap}
    componentAppliedLinearAccClamp=<1500 1500 1500>
    componentLinearAccScale=<0.4 0.4 0.55>
    componentLinearVelScale=<0.4 0.4 0.6>
    gravityOverride {x=-980}
    physicsBodyDefinitions [{
      sphereCollisionRadius=0
      boundBone {boneName=dyn_r_fr_pants_strap}
      boxExtents=<15 15 15>
      constraintSetup {angularLimitsMax=<5 5 5>, angularLimitsMin=<-5 -5 -5>, angularTarget=<1 0 0>, linearAxesMax=<1.5 1.5 2>, linearAxesMin=<-1.5 -1.5 0>}
      localJointOffset=<5 0 -2>
     }]
    relativeSpaceBone {boneName=pelvis}
   }
   n48 = "Constraint - dyn_zipper_MSTR (PPPPPPPPPP)"!(
    ComponentPose=@n47
    Alpha=1
    ConstraintWeights_0=0.1084
    ConstraintWeights_1:float=0.0051
    ConstraintWeights_2:float=0
    ConstraintWeights_3:float=0
    ConstraintWeights_4:float=0.5705
    ConstraintWeights_5:float=0.1481
    ConstraintWeights_6:float=0.0025
    ConstraintWeights_7:float=0.006
    ConstraintWeights_8:float=0.0625
    ConstraintWeights_9:float=0.0969
   ) -> (Pose) {
    alphaBoolBlend {blendOption=Linear}
    boneToModify {boneName=dyn_zipper_MSTR}
    constraintSetup [
     {targetBone {boneName=spine_01}}
     {targetBone {boneName=spine_02}}
     {targetBone {boneName=spine_04}}
     {targetBone {boneName=thigh_l}}
     {targetBone {boneName=pelvis}}
     {targetBone {boneName=thigh_twist_01_l}}
     {targetBone {boneName=spine_03}}
     {targetBone {boneName=thigh_twist_01_r}}
     {targetBone {boneName=dyn_jacket_fr}}
     {targetBone {boneName=dyn_jacket_fr_l}}
    ]
    constraintWeights [1,1,1,1,1,1,1,1,1,1]
   }
   n49 = "Constraint - dyn_zipper (PPPPPPPPPP)"!(
    ComponentPose=@n48
    Alpha=1
    ConstraintWeights_0=0.1084
    ConstraintWeights_1:float=0.0051
    ConstraintWeights_2:float=0
    ConstraintWeights_3:float=0
    ConstraintWeights_4:float=0.5705
    ConstraintWeights_5:float=0.1481
    ConstraintWeights_6:float=0.0025
    ConstraintWeights_7:float=0.006
    ConstraintWeights_8:float=0.0625
    ConstraintWeights_9:float=0.0969
   ) -> (Pose) {
    alphaBoolBlend {blendOption=Linear}
    boneToModify {boneName=dyn_zipper}
    constraintSetup [
     {targetBone {boneName=spine_01}}
     {targetBone {boneName=spine_02}}
     {targetBone {boneName=spine_04}}
     {targetBone {boneName=thigh_l}}
     {targetBone {boneName=pelvis}}
     {targetBone {boneName=thigh_twist_01_l}}
     {targetBone {boneName=spine_03}}
     {targetBone {boneName=thigh_twist_01_r}}
     {targetBone {boneName=dyn_jacket_fr}}
     {targetBone {boneName=dyn_jacket_fr_l}}
    ]
    constraintWeights [1,1,1,1,1,1,1,1,1,1]
   }
   n50 = RigidBody!(ComponentPose=@n49, bAlphaBoolEnabled:bool, OverrideWorldGravity:Vector="0,0,-980", ExternalForce) -> (Pose) {
    alphaInputType=Bool
    bClampLinearTranslationLimitToRefPose=true
    bOverrideWorldGravity=true
    overridePhysicsAsset=&/M_MED_SnowSoldierFashion_Physics
    simulationSpace=BaseBoneSpace
    alphaBoolBlend {blendInTime=0.1, blendOption=Linear, blendOutTime=0.1}
    baseBoneRef {boneName=root}
    componentAppliedLinearAccClamp=<3000 3000 3000>
    componentLinearAccScale=<0.3 0.3 0.5>
    componentLinearVelScale=<0.3 0.3 0.5>
   }
   n51 = "Constraint - dyn_zipper (T)"(ComponentPose=@n50, Alpha=1, ConstraintWeights_0=1) {
    alphaBoolBlend {blendOption=Linear}
    boneToModify {boneName=dyn_zipper}
    constraintSetup [{transformType=Translation, targetBone {boneName=dyn_zipper_MSTR}}]
    constraintWeights [1]
   }
   n52 = "Transform (Modify) Bone - Bone: dyn_collar_front_l_MSTR"(ComponentPose=@n51, Alpha=1, Translation="0,0,0", Rotation="5,0,0", Scale="1,1,1") {
    rotationMode=BMM_Additive
    rotationSpace=BCS_BoneSpace
    translationMode=BMM_Additive
    translationSpace=BCS_BoneSpace
    alphaBoolBlend {blendOption=Linear}
    boneToModify {boneName=dyn_collar_front_l_MSTR}
    rotation {pitch=5}
   }
   n53 = Anim Dynamics - Bone: dyn_zipper(ComponentPose=@n52, Alpha=1) {
    angularDampingOverride=0.999
    angularSpringConstant=200
    bGravityOverrideInSimSpace=true
    bOverrideAngularDamping=true
    bUseGravityOverride=true
    bUseSphericalLimits=true
    gravityScale=0.35
    linearDampingOverride=1
    linearSpringConstant=150
    simulationSpace=BoneRelative
    alphaBoolBlend {blendOption=Linear}
    boundBone {boneName=dyn_zipper}
    componentAppliedLinearAccClamp=<1500 1500 1500>
    componentLinearAccScale=<0.1 0.1 0.25>
    componentLinearVelScale=<0.3 0.3 0.4>
    gravityOverride {z=-980}
    physicsBodyDefinitions [{
      sphereCollisionRadius=0
      boundBone {boneName=dyn_zipper}
      boxExtents=<6 6 6>
      constraintSetup {angularLimitsMax=<5 60 100>, angularLimitsMin=<-5 -60 -20>, angularTarget=<1 0 0>}
      localJointOffset=<0.5 0 0>
     }]
    planarLimits [{drivingBone {boneName=pelvis}, planeTransform {location=<0 10 0>, rotation=<rot 90 -120 0>}}]
    relativeSpaceBone {boneName=root}
    sphericalLimits [
     {limitRadius=0, drivingBone {boneName=None}, sphereLocalOffset=<0 0 0>}
     {limitRadius=6, drivingBone {boneName=thigh_twist_01_l}, sphereLocalOffset=<15 0 0.8>}
    ]
   }
   n54 = "Transform (Modify) Bone - Bone: dyn_r_bk_upper_strap_mstr"!(
    ComponentPose=@n53
    bAlphaBoolEnabled:bool=Get not_crouching() -> (not_crouching:bool)
    Translation="0,0,0"
    Rotation="-5,0,0"
    Scale="1,1,1"
   ) -> (Pose) {
    alphaInputType=Bool
    rotationMode=BMM_Additive
    rotationSpace=BCS_BoneSpace
    translationMode=BMM_Additive
    translationSpace=BCS_BoneSpace
    alphaBoolBlend {blendOption=Linear}
    boneToModify {boneName=dyn_r_bk_upper_strap_mstr}
    rotation {pitch=-5}
   }
   n55 = Get ADS_UpArm_Rot_R()
   n56 = "Constraint - dyn_collar_front_r_01 (P)"(ComponentPose=@n54, Alpha=@n55, ConstraintWeights_0=1) {
    alphaBoolBlend {blendOption=Linear}
    boneToModify {boneName=dyn_collar_front_r_01}
    constraintSetup [{targetBone {boneName=neck_02}}]
    constraintWeights [1]
   }
   n57 = "Transform (Modify) Bone - Bone: dyn_collar_front_r_MSTR"(ComponentPose=@n56, Alpha=@n55, Translation="2,3,-1", Rotation="20,-50,20", Scale="1,1,1") {
    rotationMode=BMM_Additive
    rotationSpace=BCS_BoneSpace
    translationMode=BMM_Additive
    translationSpace=BCS_BoneSpace
    alphaBoolBlend {blendOption=Linear}
    boneToModify {boneName=dyn_collar_front_r_MSTR}
    rotation=<rot 20 -50 20>
    translation=<2 3 -1>
   }
   n58 = "Transform (Modify) Bone - Bone: dyn_collar_front_r_MSTR"(ComponentPose=@n57, Alpha=Get ADS_UpArm_Rot(), Translation="0,0,0", Rotation="0,0,0", Scale="1,1,1") {
    rotationMode=BMM_Additive
    rotationSpace=BCS_BoneSpace
    translationMode=BMM_Additive
    translationSpace=BCS_BoneSpace
    alphaBoolBlend {blendOption=Linear}
    boneToModify {boneName=dyn_collar_front_r_MSTR}
   }
   n59 = Get moving_crouch() -> (moving_crouch:double)
   n60 = "Transform (Modify) Bone - Bone: dyn_jacket_fr_r"(ComponentPose=@n58, Alpha=@n59, Translation="0,0,1", Rotation="40,0,0", Scale="1,1,1") {
    rotationMode=BMM_Additive
    rotationSpace=BCS_BoneSpace
    translationMode=BMM_Additive
    translationSpace=BCS_BoneSpace
    alphaBoolBlend {blendOption=Linear}
    boneToModify {boneName=dyn_jacket_fr_r}
    rotation {pitch=40}
    translation {z=1}
   }
   n61 = "Transform (Modify) Bone - Bone: dyn_jacket_fr_l"(ComponentPose=@n60, Alpha=@n59, Translation="0,0,0.5", Rotation="30,0,0", Scale="1,1,1") {
    rotationMode=BMM_Additive
    rotationSpace=BCS_BoneSpace
    translationMode=BMM_Additive
    translationSpace=BCS_BoneSpace
    alphaBoolBlend {blendOption=Linear}
    boneToModify {boneName=dyn_jacket_fr_l}
    rotation {pitch=30}
    translation {z=0.5}
   }
   n62 = "Constraint - dyn_jacket_bk_mstr (PP)"!(ComponentPose=@n61, Alpha=1, ConstraintWeights_0=0.5, ConstraintWeights_1:float=0.5) -> (Pose) {
    alphaBoolBlend {blendOption=Linear}
    boneToModify {boneName=dyn_jacket_bk_mstr}
    constraintSetup [{targetBone {boneName=dyn_jacket_bk_r_mstr}}, {targetBone {boneName=dyn_jacket_bk_l_mstr}}]
    constraintWeights [1,1]
   }
   n63 = "Transform (Modify) Bone - Bone: dyn_jacket_bk_mstr"(ComponentPose=@n62, Alpha=1, Translation="-0.5,0,0", Rotation="0,0,0", Scale="1,1,1") {
    rotationMode=BMM_Additive
    rotationSpace=BCS_BoneSpace
    translationMode=BMM_Additive
    translationSpace=BCS_BoneSpace
    alphaBoolBlend {blendOption=Linear}
    boneToModify {boneName=dyn_jacket_bk_mstr}
    translation {x=-0.5}
   }
   Output Pose(Result=Component To Local(ComponentPose=@n63)) {layerGroup=None}
   Save cached pose 'Pre_Crouch'() {cacheName=Pre_Crouch}
   n64 = "Constraint - dyn_l_fr_upper_strap_mstr (PPPPPPPP)"!(
    ComponentPose=Local To Component(LocalPose=Copy Pose From Mesh() {bCopyCurves=true, bCopyCustomAttributes=true, bUseAttachedParent=true, bUseMeshPose=true})
    Alpha=1
    ConstraintWeights_0=0.0123
    ConstraintWeights_1:float=0
    ConstraintWeights_2:float=0.0062
    ConstraintWeights_3:float=0.2514
    ConstraintWeights_4:float=0.7171
    ConstraintWeights_5:float=0
    ConstraintWeights_6:float=0
    ConstraintWeights_7:float=0.0129
   ) -> (Pose) {
    alphaBoolBlend {blendOption=Linear}
    boneToModify {boneName=dyn_l_fr_upper_strap_mstr}
    constraintSetup [
     {targetBone {boneName=spine_01}}
     {targetBone {boneName=spine_02}}
     {targetBone {boneName=thigh_l}}
     {targetBone {boneName=pelvis}}
     {targetBone {boneName=thigh_twist_01_l}}
     {targetBone {boneName=thigh_r}}
     {targetBone {boneName=spine_03}}
     {targetBone {boneName=thigh_twist_01_r}}
    ]
    constraintWeights [1,1,1,1,1,1,1,1]
   }
   n65 = "Constraint - dyn_r_fr_upper_strap_mstr (PPPPPPPP)"!(
    ComponentPose=@n64
    Alpha=1
    ConstraintWeights_0=0.0005
    ConstraintWeights_1:float=0
    ConstraintWeights_2:float=0
    ConstraintWeights_3:float=0.0767
    ConstraintWeights_4:float=0.0009
    ConstraintWeights_5:float=0.0008
    ConstraintWeights_6:float=0.9211
    ConstraintWeights_7:float=0
   ) -> (Pose) {
    alphaBoolBlend {blendOption=Linear}
    boneToModify {boneName=dyn_r_fr_upper_strap_mstr}
    constraintSetup [
     {targetBone {boneName=spine_01}}
     {targetBone {boneName=spine_02}}
     {targetBone {boneName=thigh_l}}
     {targetBone {boneName=pelvis}}
     {targetBone {boneName=thigh_twist_01_l}}
     {targetBone {boneName=thigh_r}}
     {targetBone {boneName=thigh_twist_01_r}}
     {targetBone {boneName=calf_r}}
    ]
    constraintWeights [1,1,1,1,1,1,1,1]
   }
   n66 = "Constraint - dyn_r_bk_upper_strap_mstr (PPPPPP)"!(
    ComponentPose=@n65
    Alpha=1
    ConstraintWeights_0=0.0004
    ConstraintWeights_1:float=0
    ConstraintWeights_2:float=0.6947
    ConstraintWeights_3:float=0.0076
    ConstraintWeights_4:float=0.0019
    ConstraintWeights_5:float=0.2954
   ) -> (Pose) {
    alphaBoolBlend {blendOption=Linear}
    boneToModify {boneName=dyn_r_bk_upper_strap_mstr}
    constraintSetup [
     {targetBone {boneName=spine_01}}
     {targetBone {boneName=thigh_l}}
     {targetBone {boneName=pelvis}}
     {targetBone {boneName=thigh_twist_01_l}}
     {targetBone {boneName=thigh_r}}
     {targetBone {boneName=thigh_twist_01_r}}
    ]
    constraintWeights [1,1,1,1,1,1]
   }
   n67 = "Constraint - dyn_l_bk_upper_strap_mstr (PPPPP)"!(
    ComponentPose=@n66
    Alpha=1
    ConstraintWeights_0=0.0039
    ConstraintWeights_1:float=0.0015
    ConstraintWeights_2:float=0.6851
    ConstraintWeights_3:float=0.2989
    ConstraintWeights_4:float=0.0106
   ) -> (Pose) {
    alphaBoolBlend {blendOption=Linear}
    boneToModify {boneName=dyn_l_bk_upper_strap_mstr}
    constraintSetup [
     {targetBone {boneName=spine_01}}
     {targetBone {boneName=thigh_l}}
     {targetBone {boneName=pelvis}}
     {targetBone {boneName=thigh_twist_01_l}}
     {targetBone {boneName=thigh_twist_01_r}}
    ]
    constraintWeights [1,1,1,1,1]
   }
   n68 = "Constraint - dyn_jacket_fr_r_mstr (PPPPPPPPP)"!(
    ComponentPose=@n67
    Alpha=1
    ConstraintWeights_0=0.0033
    ConstraintWeights_1:float=0
    ConstraintWeights_2:float=0
    ConstraintWeights_3:float=0.1043
    ConstraintWeights_4:float=0.0013
    ConstraintWeights_5:float=0.0004
    ConstraintWeights_6:float=0
    ConstraintWeights_7:float=0.8906
    ConstraintWeights_8:float=0
   ) -> (Pose) {
    alphaBoolBlend {blendOption=Linear}
    boneToModify {boneName=dyn_jacket_fr_r_mstr}
    constraintSetup [
     {targetBone {boneName=spine_01}}
     {targetBone {boneName=spine_02}}
     {targetBone {boneName=thigh_l}}
     {targetBone {boneName=pelvis}}
     {targetBone {boneName=thigh_twist_01_l}}
     {targetBone {boneName=thigh_r}}
     {targetBone {boneName=spine_03}}
     {targetBone {boneName=thigh_twist_01_r}}
     {targetBone {boneName=calf_r}}
    ]
    constraintWeights [1,1,1,1,1,1,1,1,1]
   }
   n69 = "Constraint - dyn_jacket_fr_l_mstr (PPPPPPPP)"!(
    ComponentPose=@n68
    Alpha=1
    ConstraintWeights_0=0.0161
    ConstraintWeights_1:float=0
    ConstraintWeights_2:float=0.0042
    ConstraintWeights_3:float=0.2764
    ConstraintWeights_4:float=0.6908
    ConstraintWeights_5:float=0
    ConstraintWeights_6:float=0
    ConstraintWeights_7:float=0.0124
   ) -> (Pose) {
    alphaBoolBlend {blendOption=Linear}
    boneToModify {boneName=dyn_jacket_fr_l_mstr}
    constraintSetup [
     {targetBone {boneName=spine_01}}
     {targetBone {boneName=spine_02}}
     {targetBone {boneName=thigh_l}}
     {targetBone {boneName=pelvis}}
     {targetBone {boneName=thigh_twist_01_l}}
     {targetBone {boneName=thigh_r}}
     {targetBone {boneName=spine_03}}
     {targetBone {boneName=thigh_twist_01_r}}
    ]
    constraintWeights [1,1,1,1,1,1,1,1]
   }
   n70 = "Constraint - dyn_jacket_fr_r_mstr (U)"(ComponentPose=@n69, Alpha=1, ConstraintWeights_0=1) {
    alphaBoolBlend {blendOption=Linear}
    boneToModify {boneName=dyn_jacket_fr_r_mstr}
    constraintSetup [{transformType=LookAt, targetBone {boneName=dyn_r_fr_upper_strap_mstr}}]
    constraintWeights [1]
   }
   n71 = "Constraint - dyn_jacket_fr_l_mstr (U)"(ComponentPose=@n70, Alpha=1, ConstraintWeights_0=1) {
    alphaBoolBlend {blendOption=Linear}
    boneToModify {boneName=dyn_jacket_fr_l_mstr}
    constraintSetup [{transformType=LookAt, targetBone {boneName=dyn_l_fr_upper_strap_mstr}}]
    constraintWeights [1]
   }
   n72 = "Transform (Modify) Bone - Bone: dyn_jacket_fr_r_mstr"(ComponentPose=@n71, Alpha=1, Translation="0,0,2", Rotation="25,0,-10", Scale="1,1,1") {
    rotationMode=BMM_Additive
    rotationSpace=BCS_BoneSpace
    translationMode=BMM_Additive
    translationSpace=BCS_BoneSpace
    alphaBoolBlend {blendOption=Linear}
    boneToModify {boneName=dyn_jacket_fr_r_mstr}
    rotation {pitch=25, roll=-10}
    translation {z=2}
   }
   n73 = "Transform (Modify) Bone - Bone: dyn_jacket_fr_l_mstr"(ComponentPose=@n72, Alpha=1, Translation="0,0,2", Rotation="0,0,0", Scale="1,1,1") {
    rotationMode=BMM_Additive
    rotationSpace=BCS_BoneSpace
    translationMode=BMM_Additive
    translationSpace=BCS_BoneSpace
    alphaBoolBlend {blendOption=Linear}
    boneToModify {boneName=dyn_jacket_fr_l_mstr}
    translation {z=2}
   }
   n74 = "Constraint - dyn_jacket_fr_mstr (PP)"!(ComponentPose=@n73, Alpha=1, ConstraintWeights_0=0.5, ConstraintWeights_1:float=0.5) -> (Pose) {
    alphaBoolBlend {blendOption=Linear}
    boneToModify {boneName=dyn_jacket_fr_mstr}
    constraintSetup [{targetBone {boneName=dyn_jacket_fr_l_mstr}}, {targetBone {boneName=dyn_jacket_fr_r_mstr}}]
    constraintWeights [1,1]
   }
   n75 = "Constraint - dyn_jacket_bk_mstr (PP)"(ComponentPose=@n74, Alpha=1, ConstraintWeights_0=0.5, ConstraintWeights_1=0.5) {
    alphaBoolBlend {blendOption=Linear}
    boneToModify {boneName=dyn_jacket_bk_mstr}
    constraintSetup [{targetBone {boneName=dyn_jacket_bk_l_mstr}}, {targetBone {boneName=dyn_jacket_bk_r_mstr}}]
    constraintWeights [1,1]
   }
   n76 = "Constraint - dyn_jacket_l_mstr (PP)"!(ComponentPose=@n75, Alpha=1, ConstraintWeights_0=0.5, ConstraintWeights_1:float=0.5) -> (Pose) {
    alphaBoolBlend {blendOption=Linear}
    boneToModify {boneName=dyn_jacket_l_mstr}
    constraintSetup [{targetBone {boneName=dyn_jacket_fr_l_mstr}}, {targetBone {boneName=dyn_jacket_bk_l_mstr}}]
    constraintWeights [1,1]
   }
   n77 = "Constraint - dyn_jacket_r_mstr (PP)"!(ComponentPose=@n76, Alpha=1, ConstraintWeights_0=0.5, ConstraintWeights_1:float=0.5) -> (Pose) {
    alphaBoolBlend {blendOption=Linear}
    boneToModify {boneName=dyn_jacket_r_mstr}
    constraintSetup [{targetBone {boneName=dyn_jacket_fr_r_mstr}}, {targetBone {boneName=dyn_jacket_bk_r_mstr}}]
    constraintWeights [1,1]
   }
   n78 = "Constraint - dyn_r_fr_upper_strap_mstr (R)"(ComponentPose=@n77, Alpha=1, ConstraintWeights_0=1) {
    alphaBoolBlend {blendOption=Linear}
    boneToModify {boneName=dyn_r_fr_upper_strap_mstr}
    constraintSetup [{transformType=Rotation, targetBone {boneName=thigh_r}}]
    constraintWeights [1]
   }
   n79 = "Constraint - dyn_r_bk_upper_strap_mstr (R)"(ComponentPose=@n78, Alpha=1, ConstraintWeights_0=1) {
    alphaBoolBlend {blendOption=Linear}
    boneToModify {boneName=dyn_r_bk_upper_strap_mstr}
    constraintSetup [{transformType=Rotation, targetBone {boneName=thigh_r}}]
    constraintWeights [1]
   }
   n80 = "Constraint - dyn_l_fr_upper_strap_mstr (R)"(ComponentPose=@n79, Alpha=1, ConstraintWeights_0=1) {
    alphaBoolBlend {blendOption=Linear}
    boneToModify {boneName=dyn_l_fr_upper_strap_mstr}
    constraintSetup [{transformType=Rotation, targetBone {boneName=thigh_l}}]
    constraintWeights [1]
   }
   n81 = "Constraint - dyn_l_bk_upper_strap_mstr (R)"(ComponentPose=@n80, Alpha=1, ConstraintWeights_0=1) {
    alphaBoolBlend {blendOption=Linear}
    boneToModify {boneName=dyn_l_bk_upper_strap_mstr}
    constraintSetup [{transformType=Rotation, targetBone {boneName=thigh_l}}]
    constraintWeights [1]
   }
   n82 = Anim Dynamics - Bone: dyn_r_fr_upper_strap(ComponentPose=@n81, Alpha=1, ExternalForce="0,0,0") {
    angularDampingOverride=0.9
    angularSpringConstant=200
    bAngularSpring=true
    bGravityOverrideInSimSpace=true
    bOverrideAngularDamping=true
    bOverrideLinearDamping=true
    bUseGravityOverride=true
    gravityScale=0.35
    linearDampingOverride=0.999
    linearSpringConstant=150
    simulationSpace=BoneRelative
    alphaBoolBlend {blendOption=Linear}
    boundBone {boneName=dyn_r_fr_upper_strap}
    componentAppliedLinearAccClamp=<1500 1500 1500>
    componentLinearAccScale=<0.1 0.1 0.25>
    componentLinearVelScale=<0.3 0.3 0.4>
    gravityOverride {z=-980}
    physicsBodyDefinitions [{
      sphereCollisionRadius=0
      boundBone {boneName=dyn_r_fr_upper_strap}
      boxExtents=<17 17 17>
      constraintSetup {angularLimitsMax=<5 0 10>, angularLimitsMin=<-5 -50 -10>, angularTarget=<1 0 0>}
      localJointOffset=<0.75 0 0>
     }]
    relativeSpaceBone {boneName=root}
   }
   n83 = Anim Dynamics - Bone: dyn_r_bk_upper_strap(ComponentPose=@n82, Alpha=1, ExternalForce="0,0,0") {
    angularDampingOverride=0.9
    angularSpringConstant=200
    bAngularSpring=true
    bGravityOverrideInSimSpace=true
    bOverrideAngularDamping=true
    bOverrideLinearDamping=true
    bUseGravityOverride=true
    gravityScale=0.35
    linearDampingOverride=0.999
    linearSpringConstant=150
    simulationSpace=BoneRelative
    alphaBoolBlend {blendOption=Linear}
    boundBone {boneName=dyn_r_bk_upper_strap}
    componentAppliedLinearAccClamp=<1500 1500 1500>
    componentLinearAccScale=<0.1 0.1 0.25>
    componentLinearVelScale=<0.3 0.3 0.4>
    gravityOverride {z=-980}
    physicsBodyDefinitions [{
      sphereCollisionRadius=0
      boundBone {boneName=dyn_r_bk_upper_strap}
      boxExtents=<12 12 12>
      constraintSetup {angularLimitsMax=<5 0 10>, angularLimitsMin=<-5 -50 -10>, angularTarget=<1 0 0>}
      localJointOffset=<0.5 0 0>
     }]
    relativeSpaceBone {boneName=root}
   }
   n84 = Anim Dynamics - Bone: dyn_l_fr_upper_strap(ComponentPose=@n83, Alpha=1, ExternalForce="0,0,0") {
    angularDampingOverride=0.9
    angularSpringConstant=200
    bAngularSpring=true
    bGravityOverrideInSimSpace=true
    bOverrideAngularDamping=true
    bOverrideLinearDamping=true
    bUseGravityOverride=true
    gravityScale=0.35
    linearDampingOverride=0.999
    linearSpringConstant=150
    simulationSpace=BoneRelative
    alphaBoolBlend {blendOption=Linear}
    boundBone {boneName=dyn_l_fr_upper_strap}
    componentAppliedLinearAccClamp=<1500 1500 1500>
    componentLinearAccScale=<0.1 0.1 0.25>
    componentLinearVelScale=<0.3 0.3 0.4>
    gravityOverride {z=-980}
    physicsBodyDefinitions [{
      sphereCollisionRadius=0
      boundBone {boneName=dyn_l_fr_upper_strap}
      boxExtents=<17 17 17>
      constraintSetup {angularLimitsMax=<5 0 10>, angularLimitsMin=<-5 -50 -10>, angularTarget=<1 0 0>}
      localJointOffset=<0.75 0 0>
     }]
    relativeSpaceBone {boneName=root}
   }
   n85 = Anim Dynamics - Bone: dyn_l_bk_upper_strap(ComponentPose=@n84, Alpha=1, ExternalForce="0,0,0") {
    angularDampingOverride=0.9
    angularSpringConstant=200
    bAngularSpring=true
    bGravityOverrideInSimSpace=true
    bOverrideAngularDamping=true
    bOverrideLinearDamping=true
    bUseGravityOverride=true
    gravityScale=0.35
    linearDampingOverride=0.999
    linearSpringConstant=150
    simulationSpace=BoneRelative
    alphaBoolBlend {blendOption=Linear}
    boundBone {boneName=dyn_l_bk_upper_strap}
    componentAppliedLinearAccClamp=<1500 1500 1500>
    componentLinearAccScale=<0.1 0.1 0.25>
    componentLinearVelScale=<0.3 0.3 0.4>
    gravityOverride {z=-980}
    physicsBodyDefinitions [{
      sphereCollisionRadius=0
      boundBone {boneName=dyn_l_bk_upper_strap}
      boxExtents=<12 12 12>
      constraintSetup {angularLimitsMax=<5 0 10>, angularLimitsMin=<-5 -50 -10>, angularTarget=<1 0 0>}
      localJointOffset=<0.5 0 0>
     }]
    relativeSpaceBone {boneName=root}
   }
   n86 = "Transform (Modify) Bone - Bone: dyn_r_bk_upper_strap_mstr"!(ComponentPose=@n85, Alpha:float=1, Translation="0,0,0", Rotation="0,0,0", Scale="1,1,1") -> (Pose) {
    rotationMode=BMM_Additive
    rotationSpace=BCS_BoneSpace
    translationMode=BMM_Additive
    translationSpace=BCS_BoneSpace
    alphaBoolBlend {blendOption=Linear}
    boneToModify {boneName=dyn_r_bk_upper_strap_mstr}
   }
   n87 = Control Rig!(Source=Component To Local(ComponentPose=@n86), Alpha, Pickaxe_Swinging:double=Get pickaxe_swing() -> (pickaxe_swing:double)) -> (Pose) {
    alphaBoolBlend {blendOption=Linear}
    controlRigAssetReference {blueprintRigClass=&/M_MED_SnowSoldierFashion_CtrlRig.M_MED_SnowSoldierFashion_CtrlRig_C}
   }
   n88 = "Constraint - dyn_jacket_bk_mstr (PP)"(ComponentPose=Local To Component(LocalPose=@n87), Alpha=1, ConstraintWeights_0=0.5, ConstraintWeights_1=0.5) {
    alphaBoolBlend {blendOption=Linear}
    boneToModify {boneName=dyn_jacket_bk_mstr}
    constraintSetup [{targetBone {boneName=dyn_jacket_bk_l_mstr}}, {targetBone {boneName=dyn_jacket_bk_r_mstr}}]
    constraintWeights [1,1]
   }
   n89 = Anim Dynamics - Bone: dyn_jacket_fr(ComponentPose=@n88, Alpha=1, ExternalForce="0,0,0") {
    angularDampingOverride=0.9
    angularSpringConstant=200
    bAngularSpring=true
    bGravityOverrideInSimSpace=true
    bOverrideAngularDamping=true
    bOverrideLinearDamping=true
    bUseGravityOverride=true
    gravityScale=0.35
    linearDampingOverride=0.999
    linearSpringConstant=150
    simulationSpace=BoneRelative
    alphaBoolBlend {blendOption=Linear}
    boundBone {boneName=dyn_jacket_fr}
    componentAppliedLinearAccClamp=<1500 1500 1500>
    componentLinearAccScale=<0.1 0.1 0.25>
    componentLinearVelScale=<0.3 0.3 0.4>
    gravityOverride {z=-980}
    physicsBodyDefinitions [{
      sphereCollisionRadius=0
      boundBone {boneName=dyn_jacket_fr}
      boxExtents=<15 15 15>
      constraintSetup {angularLimitsMax=<2.5 0 5>, angularLimitsMin=<-2.5 -30 -5>, angularTarget=<1 0 0>}
      localJointOffset=<0.75 0 0>
     }]
    relativeSpaceBone {boneName=root}
   }
   n90 = Anim Dynamics - Bone: dyn_jacket_fr_l(ComponentPose=@n89, Alpha=1, ExternalForce="0,0,0") {
    angularDampingOverride=0.9
    angularSpringConstant=200
    bAngularSpring=true
    bGravityOverrideInSimSpace=true
    bOverrideAngularDamping=true
    bOverrideLinearDamping=true
    bUseGravityOverride=true
    gravityScale=0.35
    linearDampingOverride=0.999
    linearSpringConstant=150
    simulationSpace=BoneRelative
    alphaBoolBlend {blendOption=Linear}
    boundBone {boneName=dyn_jacket_fr_l}
    componentAppliedLinearAccClamp=<1500 1500 1500>
    componentLinearAccScale=<0.1 0.1 0.25>
    componentLinearVelScale=<0.3 0.3 0.4>
    gravityOverride {z=-980}
    physicsBodyDefinitions [{
      sphereCollisionRadius=0
      boundBone {boneName=dyn_jacket_fr_l}
      boxExtents=<15 15 15>
      constraintSetup {angularLimitsMax=<2.5 0 5>, angularLimitsMin=<-2.5 -30 -5>, angularTarget=<1 0 0>}
      localJointOffset=<0.75 0 0>
     }]
    relativeSpaceBone {boneName=root}
   }
   n91 = Anim Dynamics - Bone: dyn_jacket_l(ComponentPose=@n90, Alpha=1, ExternalForce="0,0,0") {
    angularDampingOverride=0.9
    angularSpringConstant=200
    bAngularSpring=true
    bGravityOverrideInSimSpace=true
    bOverrideAngularDamping=true
    bOverrideLinearDamping=true
    bUseGravityOverride=true
    gravityScale=0.35
    linearDampingOverride=0.999
    linearSpringConstant=150
    simulationSpace=BoneRelative
    alphaBoolBlend {blendOption=Linear}
    boundBone {boneName=dyn_jacket_l}
    componentAppliedLinearAccClamp=<1500 1500 1500>
    componentLinearAccScale=<0.1 0.1 0.25>
    componentLinearVelScale=<0.3 0.3 0.4>
    gravityOverride {z=-980}
    physicsBodyDefinitions [{
      sphereCollisionRadius=0
      boundBone {boneName=dyn_jacket_l}
      boxExtents=<15 15 15>
      constraintSetup {angularLimitsMax=<2.5 0 5>, angularLimitsMin=<-2.5 -30 -5>, angularTarget=<1 0 0>}
      localJointOffset=<0.75 0 0>
     }]
    relativeSpaceBone {boneName=root}
   }
   n92 = Anim Dynamics - Bone: dyn_jacket_bk_l(ComponentPose=@n91, Alpha=1, ExternalForce="0,0,0") {
    angularDampingOverride=0.9
    angularSpringConstant=200
    bAngularSpring=true
    bGravityOverrideInSimSpace=true
    bOverrideAngularDamping=true
    bOverrideLinearDamping=true
    bUseGravityOverride=true
    gravityScale=0.35
    linearDampingOverride=0.999
    linearSpringConstant=150
    simulationSpace=BoneRelative
    alphaBoolBlend {blendOption=Linear}
    boundBone {boneName=dyn_jacket_bk_l}
    componentAppliedLinearAccClamp=<1500 1500 1500>
    componentLinearAccScale=<0.1 0.1 0.25>
    componentLinearVelScale=<0.3 0.3 0.4>
    gravityOverride {z=-980}
    physicsBodyDefinitions [{
      sphereCollisionRadius=0
      boundBone {boneName=dyn_jacket_bk_l}
      boxExtents=<15 15 15>
      constraintSetup {angularLimitsMax=<2.5 0 5>, angularLimitsMin=<-2.5 -30 -5>, angularTarget=<1 0 0>}
      localJointOffset=<0.75 0 0>
     }]
    relativeSpaceBone {boneName=root}
   }
   n93 = Anim Dynamics - Bone: dyn_jacket_bk(ComponentPose=@n92, Alpha=1, ExternalForce="0,0,0") {
    angularDampingOverride=0.9
    angularSpringConstant=200
    bAngularSpring=true
    bGravityOverrideInSimSpace=true
    bOverrideAngularDamping=true
    bOverrideLinearDamping=true
    bUseGravityOverride=true
    gravityScale=0.35
    linearDampingOverride=0.999
    linearSpringConstant=150
    simulationSpace=BoneRelative
    alphaBoolBlend {blendOption=Linear}
    boundBone {boneName=dyn_jacket_bk}
    componentAppliedLinearAccClamp=<1500 1500 1500>
    componentLinearAccScale=<0.1 0.1 0.25>
    componentLinearVelScale=<0.3 0.3 0.4>
    gravityOverride {z=-980}
    physicsBodyDefinitions [{
      sphereCollisionRadius=0
      boundBone {boneName=dyn_jacket_bk}
      boxExtents=<15 15 15>
      constraintSetup {angularLimitsMax=<2.5 0 5>, angularLimitsMin=<-2.5 -30 -5>, angularTarget=<1 0 0>}
      localJointOffset=<0.75 0 0>
     }]
    relativeSpaceBone {boneName=root}
   }
   n94 = Anim Dynamics - Bone: dyn_jacket_bk_r(ComponentPose=@n93, Alpha=1, ExternalForce="0,0,0") {
    angularDampingOverride=0.9
    angularSpringConstant=200
    bAngularSpring=true
    bGravityOverrideInSimSpace=true
    bOverrideAngularDamping=true
    bOverrideLinearDamping=true
    bUseGravityOverride=true
    gravityScale=0.35
    linearDampingOverride=0.999
    linearSpringConstant=150
    simulationSpace=BoneRelative
    alphaBoolBlend {blendOption=Linear}
    boundBone {boneName=dyn_jacket_bk_r}
    componentAppliedLinearAccClamp=<1500 1500 1500>
    componentLinearAccScale=<0.1 0.1 0.25>
    componentLinearVelScale=<0.3 0.3 0.4>
    gravityOverride {z=-980}
    physicsBodyDefinitions [{
      sphereCollisionRadius=0
      boundBone {boneName=dyn_jacket_bk_r}
      boxExtents=<15 15 15>
      constraintSetup {angularLimitsMax=<2.5 0 5>, angularLimitsMin=<-2.5 -30 -5>, angularTarget=<1 0 0>}
      localJointOffset=<0.75 0 0>
     }]
    relativeSpaceBone {boneName=root}
   }
   n95 = Anim Dynamics - Bone: dyn_jacket_r(ComponentPose=@n94, Alpha=1, ExternalForce="0,0,0") {
    angularDampingOverride=0.9
    angularSpringConstant=200
    bAngularSpring=true
    bGravityOverrideInSimSpace=true
    bOverrideAngularDamping=true
    bOverrideLinearDamping=true
    bUseGravityOverride=true
    gravityScale=0.35
    linearDampingOverride=0.999
    linearSpringConstant=150
    simulationSpace=BoneRelative
    alphaBoolBlend {blendOption=Linear}
    boundBone {boneName=dyn_jacket_r}
    componentAppliedLinearAccClamp=<1500 1500 1500>
    componentLinearAccScale=<0.1 0.1 0.25>
    componentLinearVelScale=<0.3 0.3 0.4>
    gravityOverride {z=-980}
    physicsBodyDefinitions [{
      sphereCollisionRadius=0
      boundBone {boneName=dyn_jacket_r}
      boxExtents=<15 15 15>
      constraintSetup {angularLimitsMax=<2.5 0 5>, angularLimitsMin=<-2.5 -30 -5>, angularTarget=<1 0 0>}
      localJointOffset=<0.75 0 0>
     }]
    relativeSpaceBone {boneName=root}
   }
   n96 = Anim Dynamics - Bone: dyn_jacket_fr_r(ComponentPose=@n95, Alpha=1, ExternalForce="0,0,0") {
    angularDampingOverride=0.9
    angularSpringConstant=200
    bAngularSpring=true
    bGravityOverrideInSimSpace=true
    bOverrideAngularDamping=true
    bOverrideLinearDamping=true
    bUseGravityOverride=true
    gravityScale=0.35
    linearDampingOverride=0.999
    linearSpringConstant=150
    simulationSpace=BoneRelative
    alphaBoolBlend {blendOption=Linear}
    boundBone {boneName=dyn_jacket_fr_r}
    componentAppliedLinearAccClamp=<1500 1500 1500>
    componentLinearAccScale=<0.1 0.1 0.25>
    componentLinearVelScale=<0.3 0.3 0.4>
    gravityOverride {z=-980}
    physicsBodyDefinitions [{
      sphereCollisionRadius=0
      boundBone {boneName=dyn_jacket_fr_r}
      boxExtents=<15 15 15>
      constraintSetup {angularLimitsMax=<2.5 0 5>, angularLimitsMin=<-2.5 -30 -5>, angularTarget=<1 0 0>}
      localJointOffset=<0.75 0 0>
     }]
    relativeSpaceBone {boneName=root}
   }
   Save cached pose 'PreCrouch'(Pose=Component To Local(ComponentPose=@n96)) {cacheName=PreCrouch}
   Get Gravity Override()
  ]}
 EventGraph {nodes [
   Event BlueprintUpdateAnimation()
   > n1 = Sequence!(execute) -> (then_0, then_1, then_2:Exec, then_3:Exec, then_4:Exec, then_5:Exec, then_6:Exec, then_7:Exec)
   n2 = GetDeltaTransformFromRefPose(self=GetOwnerSkeletalMeshComponent(), BoneName=head, BaseName=spine_05)
   > Set headSpineRot(headSpineRot:double=BreakRotator(InRot=BreakTransform(InTransform=@n2).Rotation).Yaw) -> (Output_Get:double)
   Set head_rot_back(execute=@n1.then_1, head_rot_back:double=MapRangeClamped(Value=Get headSpineRot() -> (headSpineRot:double), InRangeA=-30, InRangeB=-60)) -> (Output_Get:double)
   Set Head_Rot_FAR_Back(execute=@n1.then_2, Head_Rot_FAR_Back:double=MapRangeClamped(Value=Get headSpineRot(), InRangeA=-60, InRangeB=-96)) -> (Output_Get:double)
   n3 = Gravity Override State Function DEPRECATED(
    execute=@n1.then_7
    CharacterPartAnimInstance=Self-Reference()
    GameTimeInSeconds=GetGameTimeInSeconds()
    GravityOverrideParams=Get Gravity Override Params_Pack() -> (Gravity Override Params_Pack:Gravity Override Params Struct Structure)
   )
   > Set Gravity Override(Gravity Override=@n3)
   n4 = GetDeltaTransformFromRefPose(self=GetOwnerSkeletalMeshComponent(), BoneName=upperarm_r, BaseName=spine_05)
   Set UpArm_Spine_Rot(execute=@n1.then_3, UpArm_Spine_Rot:double=BreakRotator(InRot=BreakTransform(InTransform=@n4).Rotation).Yaw) -> (Output_Get:double)
   Set ADS_UpArm_Rot(execute=@n1.then_4, ADS_UpArm_Rot:double=MapRangeClamped(Value=Get UpArm_Spine_Rot() -> (UpArm_Spine_Rot:double), InRangeA=65, InRangeB=110)) -> (Output_Get:double)
   > Set not_crouching(not_crouching:bool=NOT Boolean(A=Get bIsCrouching())) -> (Output_Get:bool)
   n5 = GetDeltaTransformFromRefPose(self=GetOwnerSkeletalMeshComponent(), BoneName=upperarm_r, BaseName=spine_05)
   Set UpArm_Spine_Rot_R(execute=@n1.then_5, UpArm_Spine_Rot_R:double=BreakRotator(InRot=BreakTransform(InTransform=@n5).Rotation).Yaw) -> (Output_Get:double)
   n6 = Get UpArm_Spine_Rot_R() -> (UpArm_Spine_Rot_R:double)
   Set ADS_UpArm_Rot_R(
    execute=@n1.then_6
    ADS_UpArm_Rot_R:double=SelectFloat(A=MapRangeClamped(Value=@n6, InRangeA=70, InRangeB=140), B=0, bPickA=Get bIsTargeting())
   ) -> (Output_Get:double)
   n7 = Get bIsPlayingMeleeAnim()
   > Set pickaxe_swing(pickaxe_swing:double=FInterpTo(Current=Get pickaxe_swing(), Target=SelectFloat(A=1, B=0, bPickA=@n7), DeltaTime=0.5, InterpSpeed=0.6)) -> (Output_Get:double)
   n8 = OR Boolean(A=AND Boolean(A=Get bIsCrouching(), B=Get bIsMoving2D()), B=Get bIsSliding())
   > Set moving_crouch(moving_crouch:double=FInterpTo(Current=Get moving_crouch(), Target=SelectFloat(A=1, B=0, bPickA=@n8), DeltaTime=0.5, InterpSpeed=0.6)) -> (Output_Get:double)
   n9 = Get bIsJumping()
   > Set leg_strap_alpha_r(leg_strap_alpha_r:double=FInterpTo(
     Current=Get leg_strap_alpha_r() -> (leg_strap_alpha_r:double)
     Target=SelectFloat(A=0.35, B=0.75, bPickA=@n9)
     DeltaTime=0.5
     InterpSpeed=0.5
    )) -> (Output_Get:double)
   n10 = GetDeltaTransformFromRefPose(self=GetOwnerSkeletalMeshComponent(), BoneName=upperarm_r, BaseName=clavicle_r)
   MapRangeClamped(Value=BreakRotator(InRot=BreakTransform(InTransform=@n10).Rotation).Yaw, InRangeA=95, InRangeB=120)
   "Absolute (Float)"(A=@n6)
  ]}
]
variables [
 ADS_collar_r {type=boolean}
 ADS_UpArm_Rot {type=number}
 ADS_UpArm_Rot_R {type=number}
 Gravity Override {type=Vector}
 Gravity Override Params_Pack {type=GravityOverrideParamsStruct}
 head_rot_back {type=number}
 Head_Rot_FAR_Back {type=number}
 headSpineRot {type=number}
 leg_strap_alpha {type=number, default=0.75}
 leg_strap_alpha_r {type=number, default=0.75}
 moving_crouch {type=number}
 NewVar {type=boolean}
 not_crouching {type=boolean}
 pickaxe_swing {type=number}
 UpArm_Spine_Rot {type=number}
 UpArm_Spine_Rot_R {type=number}
]
