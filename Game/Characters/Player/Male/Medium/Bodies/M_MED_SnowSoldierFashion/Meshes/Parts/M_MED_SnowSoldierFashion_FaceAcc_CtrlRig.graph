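# Control Rig for the male medium SnowSoldierFashion face accessory, adjusting the two head-side buckle bones as the head moves.
# Two spherical pose readers measure head orientation against neck_01 in differently offset regions, both with debug drawing on.
# Their outputs weight transform offsets on the buckle master bones: both readers drive the right buckle, only the second drives the left.
# Hood bones exist in the hierarchy but are not driven.
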ControlRigBlueprint /Game/Characters/Player/Male/Medium/Bodies/M_MED_SnowSoldierFashion/Meshes/Parts/M_MED_SnowSoldierFashion_FaceAcc_CtrlRig
graphs [
 RigVMFunctionLibrary {links [], nodes []}
 RigVMModel {nodes [
   RigUnit_BeginExecution = &RigUnit_BeginExecution()
   n1 = &RigUnit_SphericalPoseReader(
    ActiveRegionScaleFactors=(PositiveWidth=4,NegativeWidth=4,PositiveHeight=4,NegativeHeight=1)
    ActiveRegionSize=0.05
    Debug=(bDrawDebug=true,bDraw2D=false,DebugScale=25,DebugSegments=20,DebugThickness=0.25)
    DriverAxis=(X=0,Y=1,Z=0)
    DriverItem=(Type=Bone,Name="head")
    FalloffRegionScaleFactors=(PositiveWidth=0.4,NegativeWidth=0.35,PositiveHeight=0.5,NegativeHeight=0.5)
    FalloffSize=0.5
    FlipHeightScaling=false
    FlipWidthScaling=false
    OptionalParentItem=(Type=Bone,Name="neck_01")
    RotationOffset=(X=-60,Y=0,Z=130)
    ExecutePin=@RigUnit_BeginExecution.ExecutePin
   )
   > n2 = &RigUnit_SphericalPoseReader(
    ActiveRegionScaleFactors=(PositiveWidth=4,NegativeWidth=4,PositiveHeight=1,NegativeHeight=4)
    ActiveRegionSize=0.05
    Debug=(bDrawDebug=true,bDraw2D=false,DebugScale=25,DebugSegments=20,DebugThickness=0.25)
    DriverAxis=(X=0,Y=1,Z=0)
    DriverItem=(Type=Bone,Name="head")
    FalloffRegionScaleFactors=(PositiveWidth=0.4,NegativeWidth=0.35,PositiveHeight=0.5,NegativeHeight=0.5)
    FalloffSize=0.5
    FlipHeightScaling=false
    FlipWidthScaling=false
    OptionalParentItem=(Type=Bone,Name="neck_01")
    RotationOffset=(X=-120,Y=0,Z=130)
   )
   > &RigUnit_ModifyTransforms(
    ItemToModify=((Item=(Type=Bone,Name="dyn_buckle_r_mstr"),Transform=(Rotation=(X=-0.1294,Y=0.017,Z=-0.1294,W=0.983),Translation=(X=-0.2,Y=-0.2,Z=0),Scale3D=(X=1,Y=1,Z=1))))
    WeightMinimum=0
    Weight=@n2.OutputParam
   )
   > &RigUnit_ModifyTransforms(
    ItemToModify=((Item=(Type=Bone,Name="dyn_buckle_r_mstr"),Transform=(Rotation=(X=0,Y=0,Z=-0.1305,W=0.9914),Translation=(X=-0.2,Y=-0.5,Z=0),Scale3D=(X=1,Y=1,Z=1))))
    WeightMinimum=0
    Weight=@n1.OutputParam
   )
   > &RigUnit_ModifyTransforms(
    ItemToModify=((Item=(Type=Bone,Name="dyn_buckle_l_mstr"),Transform=(Rotation=(X=0,Y=0,Z=-0.1736,W=0.9848),Translation=(X=0.2,Y=-1,Z=0.5),Scale3D=(X=1,Y=1,Z=1))))
    WeightMinimum=0
    Weight=@n2.OutputParam
   )
  ]}
]
hierarchy [
 root {loc=<0 0 0>, rot=<0 0 0 1>}
 pelvis {parent=root, loc=<0 -0.7071 93.1191>, rot=<0.0127 0.707 -0.0127 -0.707>}
 spine_01 {parent=pelvis, loc=<6.7494 -0.6388 0>, rot=<0 0 0.034 -0.9994>}
 spine_02 {parent=spine_01, loc=<7.2518 -0.035 0>, rot=<0 0 0.0238 -0.9997>}
 spine_03 {parent=spine_02, loc=<6.0104 0 0>, rot=<0 0 0.0149 -0.9999>}
 spine_04 {parent=spine_03, loc=<8.1024 0 0>, rot=<0 0 -0.0132 -0.9999>}
 spine_05 {parent=spine_04, loc=<10.0146 0 0>, rot=<0 0 -0.0015 -1>}
 neck_01 {parent=spine_05, loc=<11.3138 0.1529 0>, rot=<0 0 -0.2003 -0.9797>}
 neck_02 {parent=neck_01, loc=<4.5627 0 0>, rot=<0 0 0.0504 -0.9987>}
 head {parent=neck_02, loc=<5.3914 0 0>, rot=<0 0 0.115 -0.9934>}
 faceAttach {parent=head, loc=<0.6201 0.9296 0>, rot=<0.0082 -0.7071 -0.0082 -0.7071>, scale=<0.89 0.89 0.89>}
 C_brow_mid {parent=faceAttach, loc=<0.0026 9.6516 5.0689>, rot=<0 0.0264 -0.0052 -0.9996>}
 L_brow_mid {parent=faceAttach, loc=<3.2144 9.2014 5.9725>, rot=<-0.0002 -0.0049 0.2553 -0.9669>}
 L_brow_outer {parent=faceAttach, loc=<5.3148 7.3235 5.8512>, rot=<0.0527 0.0624 0.4289 -0.8997>}
 L_cheek_inner {parent=faceAttach, loc=<4.2335 7.7652 0.1105>, rot=<-0.0672 -0.2612 0.3196 -0.9084>}
 C_nose_bridge {parent=faceAttach, loc=<0 11.1254 1.8962>, rot=<-0.2218 0.0035 0.001 -0.9751>}
 R_brow_mid {parent=faceAttach, loc=<-3.2142 9.2015 5.9726>, rot=<-0.0001 0.0043 -0.2553 -0.9669>}
 R_brow_outer {parent=faceAttach, loc=<-5.3148 7.3235 5.8512>, rot=<0.0201 -0.0214 -0.4379 -0.8985>}
 R_cheek_inner {parent=faceAttach, loc=<-4.2335 7.7652 0.1105>, rot=<-0.0681 0.2634 -0.3193 -0.9078>}
 R_eye {parent=faceAttach, loc=<-3.0117 6.7356 4.4372>, rot=<0 0 0.9998 -0.0187>}
 R_eye_lid_upper_mid {parent=R_eye, loc=<-0.0396 -1.1714 0.2631>, rot=<0.0523 -0.0009 0.0167 -0.9985>}
 R_eye_lid_lower_mid {parent=R_eye, loc=<-0.0085 -0.9425 -0.7051>, rot=<0.0329 -0.0005 0.0167 -0.9993>}
 L_eye {parent=faceAttach, loc=<3.0117 6.7356 4.4372>, rot=<0 0 0.9998 0.0187>}
 L_eye_lid_upper_mid {parent=L_eye, loc=<0.0251 -1.1623 0.2388>, rot=<0.0523 0 0 -0.9986>}
 L_eye_lid_lower_mid {parent=L_eye, loc=<0.0223 -0.923 -0.7488>, rot=<0.0329 0 0 -0.9995>}
 dyn_hood_top_mstr {parent=head, loc=<15.9081 4.7898 0>, rot=<-0.4943 -0.5057 -0.5057 -0.4943>}
 dyn_hood_top {parent=dyn_hood_top_mstr, loc=<0 0 0>, rot=<0 0 0 1>}
 dyn_hood_l_mstr {parent=head, loc=<2.4579 4.2083 -9.25>, rot=<0.0361 -0.9983 0.0441 0.0094>}
 dyn_hood_l {parent=dyn_hood_l_mstr, loc=<0 0 0>, rot=<0 0 0 1>}
 dyn_hood_r_mstr {parent=head, loc=<2.4579 4.2083 9.2564>, rot=<0.0094 0.0441 0.9983 -0.0361>}
 dyn_hood_r {parent=dyn_hood_r_mstr, loc=<0 0 0>, rot=<0 0 0 1>}
 dyn_hood_fr_mstr {parent=head, loc=<-2.7832 10.6781 0>, rot=<-0.1022 0.6997 0.6997 -0.1022>}
 dyn_hood_fr {parent=dyn_hood_fr_mstr, loc=<0 0 0>, rot=<0 0 0 1>}
 dyn_buckle_r_mstr {parent=head, loc=<1.2614 2.1652 10.1224>, rot=<0.0029 -0.7071 -0.0029 -0.7071>}
 dyn_buckle_r {parent=dyn_buckle_r_mstr, loc=<0 0 0>, rot=<0 0 0 1>}
 dyn_buckle_l_mstr {parent=head, loc=<1.2614 2.1652 -9.9752>, rot=<0.0029 -0.7071 -0.0029 -0.7071>}
 dyn_buckle_l {parent=dyn_buckle_l_mstr, loc=<0 0 0>, rot=<0 0 0 1>}
 clavicle_l {parent=spine_05, loc=<4.3575 -0.1466 -2.3223>, rot=<0.0465 -0.7289 -0.009 -0.6829>}
 upperarm_l {parent=clavicle_l, loc=<15.5111 0 0>, rot=<0.0226 -0.3234 -0.0088 -0.9459>}
 upperarm_twist_01_l {parent=upperarm_l, loc=<0.0041 0 0>, rot=<0 0 0 -1>}
 clavicle_r {parent=spine_05, loc=<4.3575 -0.1466 2.3223>, rot=<0.7289 0.0465 -0.6829 0.009>}
 upperarm_r {parent=clavicle_r, loc=<-15.5111 0 0>, rot=<0.0226 -0.3234 -0.0088 -0.9459>}
 upperarm_twist_01_r {parent=upperarm_r, loc=<-0.0041 0 0>, rot=<0 0 0 -1>}
]
