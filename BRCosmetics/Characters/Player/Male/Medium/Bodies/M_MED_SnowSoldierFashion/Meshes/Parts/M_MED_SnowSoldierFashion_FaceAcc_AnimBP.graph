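# Anim Blueprint for the face accessory part of the male SnowSoldierFashion cosmetic, animating its hood and buckles.
# Two Copy Pose From Mesh nodes blend per bone from faceAttach, then chained AnimDynamics simulate four hood bones.
# Additive bone offsets and weighted multi-target constraints place the hood front and buckles before a Control Rig finishes the pose.
# Initialization caches mesh and pawn references and computes a gravity override, which the AnimGraph never reads.
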AnimBlueprint /BRCosmetics/Characters/Player/Male/Medium/Bodies/M_MED_SnowSoldierFashion/Meshes/Parts/M_MED_SnowSoldierFashion_FaceAcc_AnimBP
target_skeleton=/Game/Characters/Player/Male/Medium/Bodies/M_MED_SnowSoldierFashion/Meshes/Parts/M_MED_SnowSoldierFashion_FaceAcc_Skeleton
graphs [
 AnimGraph {nodes [
   Blend Bone By Channel!(
    A:Pose Link Structure=Copy Pose From Mesh() {bCopyCurves=true, bCopyCustomAttributes=true, bUseAttachedParent=true, bUseMeshPose=true}
    B:Pose Link Structure
    Alpha:float=0
   ) -> (Pose:Pose Link Structure) {is=title AnimGraphNode_BlendBoneByChannel}
   Layered blend per bone() {layerSetup [{branchFilters [
       {boneName=faceAttach}
       {boneName=C_brow_mid}
       {boneName=C_nose_bridge}
       {boneName=L_brow_mid}
       {boneName=None}
       {boneName=None}
       {boneName=None}
       {boneName=None}
       {boneName=None}
       {boneName=None}
       {boneName=None}
       {boneName=None}
       {boneName=None}
      ]}]}
   n1 = Copy Pose From Mesh() {bCopyCurves=true, bCopyCustomAttributes=true, bUseAttachedParent=true, bUseMeshPose=true}
   n2 = Copy Pose From Mesh(SourceMeshComponent=Get Skel_Mesh()) {bUseAttachedParent=true}
   n3 = Anim Dynamics - Bone: dyn_hood_top!(
    ComponentPose=Local To Component(LocalPose=Layered blend per bone(BasePose=@n1, BlendPoses_0=@n2) {layerSetup [{branchFilters [{blendDepth=50, boneName=faceAttach}]}]})
    Alpha=1
    GravityOverride:Vector="0,0,-980"
    ExternalForce="0,0,0"
   ) -> (Pose) {
    angularDampingOverride=1
    angularSpringConstant=50
    bGravityOverrideInSimSpace=true
    bOverrideAngularDamping=true
    bOverrideLinearDamping=true
    bUseGravityOverride=true
    gravityScale=0.3
    linearDampingOverride=1
    linearSpringConstant=50
    simulationSpace=BoneRelative
    alphaBoolBlend {blendOption=Linear}
    boundBone {boneName=dyn_hood_top}
    componentAppliedLinearAccClamp=<2500 2500 2500>
    componentLinearAccScale=<0.2 0.2 0.2>
    componentLinearVelScale=<0.2 0.2 0.2>
    gravityOverride {z=500}
    physicsBodyDefinitions [{
      sphereCollisionRadius=0
      boundBone {boneName=dyn_hood_top}
      boxExtents=<20 20 20>
      constraintSetup {
       angularLimitsMax=<10 10 10>
       angularLimitsMin=<-10 -10 -10>
       angularTarget=<0 0 0>
       linearAxesMax=<0.5 0.5 0.5>
       linearAxesMin=<-0.5 -0.5 -0.5>
      }
      localJointOffset=<0 0 -2>
     }]
    relativeSpaceBone {boneName=root}
   }
   n4 = Anim Dynamics - Bone: dyn_hood_l!(ComponentPose=@n3, Alpha=1, GravityOverride:Vector="0,0,-980", ExternalForce="0,0,0") -> (Pose) {
    angularDampingOverride=1
    angularSpringConstant=50
    bGravityOverrideInSimSpace=true
    bOverrideAngularDamping=true
    bOverrideLinearDamping=true
    bUseGravityOverride=true
    gravityScale=0.3
    linearDampingOverride=1
    linearSpringConstant=50
    simulationSpace=BoneRelative
    alphaBoolBlend {blendOption=Linear}
    boundBone {boneName=dyn_hood_l}
    componentAppliedLinearAccClamp=<2500 2500 2500>
    componentLinearAccScale=<0.2 0.2 0.2>
    componentLinearVelScale=<0.2 0.2 0.2>
    gravityOverride {z=500}
    physicsBodyDefinitions [{
      sphereCollisionRadius=0
      boundBone {boneName=dyn_hood_l}
      boxExtents=<20 20 20>
      constraintSetup {
       angularLimitsMax=<10 10 5>
       angularLimitsMin=<-10 -10 -5>
       angularTarget=<0 0 0>
       linearAxesMax=<0.5 0.5 0.5>
       linearAxesMin=<-0.5 -0.5 -0.5>
      }
      localJointOffset=<2 0 0>
     }]
    relativeSpaceBone {boneName=root}
   }
   n5 = Anim Dynamics - Bone: dyn_hood_r!(ComponentPose=@n4, Alpha=1, GravityOverride:Vector="0,0,-980", ExternalForce="0,0,0") -> (Pose) {
    angularDampingOverride=1
    angularSpringConstant=50
    bGravityOverrideInSimSpace=true
    bOverrideAngularDamping=true
    bOverrideLinearDamping=true
    bUseGravityOverride=true
    gravityScale=0.3
    linearDampingOverride=1
    linearSpringConstant=50
    simulationSpace=BoneRelative
    alphaBoolBlend {blendOption=Linear}
    boundBone {boneName=dyn_hood_r}
    componentAppliedLinearAccClamp=<2500 2500 2500>
    componentLinearAccScale=<0.2 0.2 0.2>
    componentLinearVelScale=<0.2 0.2 0.2>
    gravityOverride {z=500}
    physicsBodyDefinitions [{
      sphereCollisionRadius=0
      boundBone {boneName=dyn_hood_r}
      boxExtents=<20 20 20>
      constraintSetup {
       angularLimitsMax=<10 10 5>
       angularLimitsMin=<-10 -10 -5>
       angularTarget=<0 0 0>
       linearAxesMax=<0.5 0.5 0.5>
       linearAxesMin=<-0.5 -0.5 -0.5>
      }
      localJointOffset=<2 0 0>
     }]
    relativeSpaceBone {boneName=root}
   }
   n6 = Anim Dynamics - Bone: dyn_hood_fr!(ComponentPose=@n5, Alpha=1, GravityOverride:Vector="0,0,-980", ExternalForce="0,0,0") -> (Pose) {
    angularDampingOverride=1
    angularSpringConstant=50
    bGravityOverrideInSimSpace=true
    bOverrideAngularDamping=true
    bOverrideLinearDamping=true
    bUseGravityOverride=true
    gravityScale=0.3
    linearDampingOverride=1
    linearSpringConstant=50
    simulationSpace=BoneRelative
    alphaBoolBlend {blendOption=Linear}
    boundBone {boneName=dyn_hood_fr}
    componentAppliedLinearAccClamp=<2500 2500 2500>
    componentLinearAccScale=<0.2 0.2 0.2>
    componentLinearVelScale=<0.2 0.2 0.2>
    gravityOverride {z=500}
    physicsBodyDefinitions [{
      sphereCollisionRadius=0
      boundBone {boneName=dyn_hood_fr}
      boxExtents=<20 20 20>
      constraintSetup {angularLimitsMax=<3 10 7>, angularLimitsMin=<-3 -10 -7>, angularTarget=<0 0 0>, linearAxesMax=<0.5 0.5 1>, linearAxesMin=<-0.5 -0.5 0>}
      localJointOffset=<2 0 0>
     }]
    relativeSpaceBone {boneName=root}
   }
   n7 = "Transform (Modify) Bone - Bone: dyn_hood_fr_mstr"(ComponentPose=@n6, Alpha=1, Translation="0,0,0.5", Rotation="0,0,0", Scale="1,1,1") {
    rotationMode=BMM_Additive
    rotationSpace=BCS_BoneSpace
    translationMode=BMM_Additive
    translationSpace=BCS_BoneSpace
    alphaBoolBlend {blendOption=Linear}
    boneToModify {boneName=dyn_hood_fr_mstr}
    translation {z=0.5}
   }
   n8 = "Constraint - dyn_buckle_l_mstr (PPPPPPPPPPPP)"!(
    ComponentPose=@n7
    Alpha=1
    ConstraintWeights_0=0.0001
    ConstraintWeights_1:float=0.0345
    ConstraintWeights_2:float=0.0031
    ConstraintWeights_3:float=0
    ConstraintWeights_4:float=0.0587
    ConstraintWeights_5:float=0.5399
    ConstraintWeights_6:float=0.0872
    ConstraintWeights_7:float=0.0001
    ConstraintWeights_8:float=0
    ConstraintWeights_9:float=0.2766
    ConstraintWeights_10:float=0
    ConstraintWeights_11:float=0
   ) -> (Pose) {
    alphaBoolBlend {blendOption=Linear}
    boneToModify {boneName=dyn_buckle_l_mstr}
    constraintSetup [
     {targetBone {boneName=spine_04}}
     {targetBone {boneName=spine_05}}
     {targetBone {boneName=clavicle_l}}
     {targetBone {boneName=clavicle_r}}
     {targetBone {boneName=neck_01}}
     {targetBone {boneName=head}}
     {targetBone {boneName=neck_02}}
     {targetBone {boneName=dyn_hood_fr}}
     {targetBone {boneName=dyn_hood_r}}
     {targetBone {boneName=dyn_hood_l}}
     {targetBone {boneName=spine_03}}
     {targetBone {boneName=upperarm_twist_01_l}}
    ]
    constraintWeights [1,1,1,1,1,1,1,1,1,1,1,1]
   }
   n9 = "Constraint - dyn_buckle_r_mstr (PPPPPPPPPPPPPP)"!(
    ComponentPose=@n8
    Alpha=1
    ConstraintWeights_0=0.0002
    ConstraintWeights_1:float=0.0393
    ConstraintWeights_2:float=0
    ConstraintWeights_3:float=0.0023
    ConstraintWeights_4:float=0.0365
    ConstraintWeights_5:float=0.6273
    ConstraintWeights_6:float=0.0538
    ConstraintWeights_7:float=0.0021
    ConstraintWeights_8:float=0.2385
    ConstraintWeights_9:float=0
    ConstraintWeights_10:float=0
    ConstraintWeights_11:float=0
    ConstraintWeights_12:float=0
    ConstraintWeights_13:float=0
   ) -> (Pose) {
    alphaBoolBlend {blendOption=Linear}
    boneToModify {boneName=dyn_buckle_r_mstr}
    constraintSetup [
     {targetBone {boneName=spine_04}}
     {targetBone {boneName=spine_05}}
     {targetBone {boneName=clavicle_l}}
     {targetBone {boneName=clavicle_r}}
     {targetBone {boneName=neck_01}}
     {targetBone {boneName=head}}
     {targetBone {boneName=neck_02}}
     {targetBone {boneName=dyn_hood_fr}}
     {targetBone {boneName=dyn_hood_r}}
     {targetBone {boneName=dyn_hood_l}}
     {targetBone {boneName=spine_02}}
     {targetBone {boneName=spine_03}}
     {targetBone {boneName=upperarm_r}}
     {targetBone {boneName=upperarm_twist_01_r}}
    ]
    constraintWeights [1,1,1,1,1,1,1,1,1,1,1,1,1,1]
   }
   n10 = "Transform (Modify) Bone - Bone: dyn_buckle_l_mstr"(ComponentPose=@n9, Alpha=1, Translation="0.1,0,0", Rotation="0,7,0", Scale="1,1,1") {
    rotationMode=BMM_Additive
    rotationSpace=BCS_BoneSpace
    translationMode=BMM_Additive
    translationSpace=BCS_BoneSpace
    alphaBoolBlend {blendOption=Linear}
    boneToModify {boneName=dyn_buckle_l_mstr}
    rotation {yaw=7}
    translation {x=0.1}
   }
   n11 = Control Rig(Source=Component To Local(ComponentPose=@n10)) {
    alphaBoolBlend {blendOption=Linear}
    controlRigAssetReference {blueprintRigClass=/Game/Characters/Player/Male/Medium/Bodies/M_MED_SnowSoldierFashion/Meshes/Parts/M_MED_SnowSoldierFashion_FaceAcc_CtrlRig.M_MED_SnowSoldierFashion_FaceAcc_CtrlRig_C}
   }
   Output Pose(Result=@n11) {layerGroup=None}
  ]}
 EventGraph {nodes [
   Event BlueprintInitializeAnimation()
   > Set Skel_Mesh(Skel_Mesh=GetOwnerSkeletalMeshComponent())
   > n1 = Cast To PlayerPawn_Athena(Object=TryGetPawnOwner())
   > n2 = Set Pawn Owner(Pawn Owner:Fort Player Pawn Athena Object Reference=@n1) -> (Output_Get:Fort Player Pawn Athena Object Reference)
   > Set Head_Mesh(Head_Mesh:Skeletal Mesh Component Object Reference=GetSkeletalMeshForPartType(self=@n2, PartType=Head)) -> (Output_Get:Skeletal Mesh Component Object Reference)
   > n3 = Gravity Override State Function DEPRECATED(CharacterPartAnimInstance=Self-Reference(), GameTimeInSeconds=GetGameTimeInSeconds(), GravityOverrideParams=Get Gravity Override Params())
   > Set Gravity Override(Gravity Override=@n3)
   Event BlueprintUpdateAnimation()
  ]}
]
variables [
 Gravity Override {type=Vector}
 Gravity Override Params {type=GravityOverrideParamsStruct}
 Head_Mesh {type=SkeletalMeshComponent}
 Pawn Owner {type=FortPlayerPawnAthena}
 Skel_Mesh {type=SkeletalMeshComponent}
]
